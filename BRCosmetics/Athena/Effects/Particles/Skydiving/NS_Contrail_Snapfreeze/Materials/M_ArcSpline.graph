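# Translucent unlit arc material for the Snapfreeze skydiving contrail, drawing toon lightning on Niagara ribbons or mesh particles.
# Lightning texture channels blend bolt, inner glow, outer glow and end contact colors, with an isGlow switch for a glow-only look and a sine fade along U.
# World offset combines SplineThicken width with a sine bow along a dynamic-parameter direction.
# A panned noise displacement branch exists but does not feed that offset.
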Material /BRCosmetics/Athena/Effects/Particles/Skydiving/NS_Contrail_Snapfreeze/Materials/M_ArcSpline
graph {
 properties {
  bAutomaticallySetUsageInEditor=false
  bUsedWithNiagaraMeshParticles=true
  bUsedWithNiagaraRibbons=true
  blendMode=BLEND_Translucent
  shadingModel=MSM_Unlit
  twoSided=true
  usedShadingModels=Unlit
 }
 nodes [
  n1 = TextureObjectParameter() {
   group=LightningT
   parameterName=LightningT_Texture
   sortPriority=0
   texture=/Game/Effects/Fort_Effects/GameplayPulginsAssets/ToonLightning/MasterMaterial/V1/T_LightningSplineTexture
  }
  n2 = VectorParameter() {group=LightningT, parameterName=LightningT_UVControl, channelNames {a=VPan, b=UPan, g=VScale, r=UScale}, defaultValue=<rgba 1 1 0 3>}
  n3 = MaterialFunctionCall!(
   Texture=@n1
   AdditionalUVOffset
   UVControl_RGB=@n2
   UVControl_A=@n2
   "[Override]TimeInput"
   "[Override]ScaleUVByCenter"=StaticBool() {value=true}
   "[Override]UVInput"=NamedRerouteUsage()
  ) {materialFunction=/Game/Effects/Niagara/MasterMaterials/Functions/Common/FXCommon_TextureSample_V2}
  Add(B=@n3.R)
  n4 = Sine(None=ComponentMask(None=TextureCoordinate() {uTiling=0.5, vTiling=0.5}) {r=true})
  n5 = Multiply(
   A=Multiply(A=@n4, B=ScalarParameter() {defaultValue=-40, parameterName=Param_1})
   B=DynamicParameter() {parameterIndex=2, defaultValue=<rgba 0 0 0 1>, paramNames [DirX,DirY,DirZ,Param4]}.RGB
  )
  n6 = ParticleRandom()
  n7 = Multiply(A=ComponentMask(None=TextureCoordinate()) {g=true}, B=ScalarParameter() {defaultValue=0.5, parameterName=Displacement Noise UV Scale})
  n8 = ScalarParameter() {defaultValue=0.1, parameterName=Displacement Noise UV Panner Speed}
  n9 = Constant()
  n10 = TextureSample(UVs=Panner(Coordinate=AppendVector(A=@n6, B=@n7), Speed=AppendVector(A=@n8, B=@n9)) {speedX=1}) {
   samplerSource=SSM_Wrap_WorldGroupSettings
   samplerType=SAMPLERTYPE_LinearGrayscale
   texture=/BRCosmetics/Weapons/FORT_Melee/Pickaxe_Parfait_Female/FX/Textures/T_TrifleParfait_LightningNoise
  }
  Add(A=@n5, B=Multiply(A=Subtract(A=@n10.R) {constB=0.5}, B=ScalarParameter() {parameterName=Displacement Noise Intensity}))
  n11 = Min(A=@n3.R, B=ScalarParameter() {defaultValue=0.2, parameterName=B_1}) {constB=0.4}
  n12 = VectorParameter() {parameterName=GlowRefine, defaultValue=<rgba 1.5 0.5 0 0>}
  n13 = VectorParameter() {parameterName=GlowColor, defaultValue=<rgba 1.4615 0.6221 2.5 1>}
  n14 = Multiply(
   A=MaterialFunctionCall(Texture=@n11, RGB=@n12.RGB, A=@n12.RGB) {materialFunction=&MF_FX_Refine}.Result
   B=@n13.RGB
  )
  n15 = VectorParameter() {parameterName=BoltColor, defaultValue=<rgba 1.6596 0.4982 3 1>}
  n16 = VectorParameter() {parameterName=InnerGlowColor, defaultValue=<rgba 9.2411 2.8881 13 5>}
  n17 = OneMinus(None=@n3.G)
  n18 = VectorParameter() {parameterName=InnerGlowRefine, defaultValue=<rgba 1 1 1 0.5>}
  n19 = Saturate(None=Multiply(
    A=MaterialFunctionCall(Texture=@n17, RGB=@n18.RGB, A=@n18.RGB) {materialFunction=&MF_FX_Refine}.Result
    B=@n3.B
   ))
  n20 = Sine(None=ComponentMask(None=TextureCoordinate() {uTiling=0.5, vTiling=0.5}) {r=true})
  n21 = VectorParameter() {parameterName=ContactGlowRefine, defaultValue=<rgba 1 1 1 0.5>}
  n22 = LinearInterpolate(
   A=VectorParameter() {parameterName=ContactGlow, defaultValue=<rgba 0.3793 0 1 1>}.RGB
   B=LinearInterpolate(A=@n13.RGB, B=LinearInterpolate(A=@n15.RGB, B=@n16.RGB, Alpha=@n19), Alpha=@n3.B)
   Alpha=Saturate(None=MaterialFunctionCall(Texture=@n20, RGB=@n21.RGB, A=@n21.RGB) {materialFunction=&MF_FX_Refine}.Result)
  )
  MaterialFunctionCall(Emissive=Multiply(A=StaticSwitchParameter(True=@n14, False=@n22) {parameterName=isGlow}, B=ParticleColor().RGB)) {materialFunction=&MF_FX_EmissiveAdjustments}
  n23 = DynamicParameter() {
   parameterIndex=2
   defaultValue=<rgba 5 0 500 30>
   paramNames [ScaleSplineWidth, BaseWitdthOffset, TextureRandom, MoveAlignLightningDirSphereMaskSize]
  }
  MaterialFunctionCall(X=Constant(), Y=@n23.TextureRandom) {materialFunction=&MakeFloat2}
  Max(A=@n3.R) {constB=0.04}
  NamedRerouteDeclaration(None=MaterialFunctionCall(XYZ=TextureCoordinate()) {materialFunction=&Swizzle}.YX) {name=UVs, nodeColor=<rgba 0.1016 1 0 1>}
  n24 = ScalarParameter() {defaultValue=6}
  n25 = ScalarParameter() {parameterName=B}
  n26 = Multiply(
   A=LinearInterpolate(A=Subtract(A=@n24, B=@n25) {constB=5}, B=@n24, Alpha=@n4) {constA=15}
   B=DynamicParameter() {paramNames [SplineWidth,Param2,Param3,Param4]}.SplineWidth
  )
  n27 = StaticBool()
  Reroute(None=Add(
    A=Transform(None=@n5) {transformSourceType=TRANSFORMSOURCE_Local}
    B=MaterialFunctionCall(WidthBase=@n26, WidthTip=@n26, Expand U or V UV Channel=@n27) {materialFunction=&SplineThicken}.WorldPositionOffset
   ))
  n28 = VectorParameter() {parameterName=OutGlowRefine, defaultValue=<rgba 1.5 0.5 0 0>}
  n29 = StaticSwitchParameter(
   True=@n11
   False=MaterialFunctionCall(Texture=@n3.R, RGB=@n28.RGB, A=@n28.RGB) {materialFunction=&MF_FX_Refine}.Result
  ) {parameterName=isGlow}
  n30 = ParticleColor()
  n31 = Sine(None=ComponentMask(None=TextureCoordinate() {uTiling=0.5, vTiling=0.5}) {r=true})
  n32 = VectorParameter() {parameterName=GradientRefine, defaultValue=<rgba 0.5 4 1 0.5>}
  Reroute(None=Multiply(
    A=Multiply(A=@n29, B=@n30.A)
    B=MaterialFunctionCall(Texture=@n31, RGB=@n32.RGB, A=@n32.RGB) {materialFunction=&MF_FX_Refine}.Result
   ))
  TextureCoordinate()
  VectorParameter() {parameterName=OffsetDirection, defaultValue=<rgba 0 0 -4 0>}
 ]
}
