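# Unlit translucent Niagara material for the trails of the Snapfreeze skydiving contrail, drawing animated lightning bolts.
# A bolt texture pans on stepped time, warped by scrolling cloud noise and eroded along V, to form a lightning SDF.
# Smoothsteps cut bolt shape and a brighter core tinted by particle color, with sine gradients, sphere masks and near-camera fade shaping opacity.
# SplineThicken sets ribbon width, and dynamic parameters drive thickness, distortion, erosion and sharpness.
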Material /BRCosmetics/Athena/Effects/Particles/Skydiving/NS_Contrail_Snapfreeze/Materials/M_FX_Snapfreeze_Trails
graph {
 properties {
  bAutomaticallySetUsageInEditor=false
  bUsedWithNiagaraMeshParticles=true
  bUsedWithNiagaraSprites=true
  blendMode=BLEND_Translucent
  pixelDepthOffsetMode=PDOM_Legacy
  shadingModel=MSM_Unlit
  usedShadingModels=Unlit
 }
 nodes [
  n1 = DynamicParameter() {defaultValue=<rgba 1 0 0 0.5>, paramNames [ThicknessMultiplier, UV.x Index, UV.y Position, BoltShaderThickness]}
  n2 = Multiply(A=NamedRerouteUsage(), B=Multiply(A=ScalarParameter() {defaultValue=20, parameterName=Thickness}, B=@n1.ThicknessMultiplier)) {constA=1}
  MaterialFunctionCall(WidthBase=@n2, WidthTip=@n2, Expand U or V UV Channel=StaticBoolParameter() {defaultValue=true}) {materialFunction=&SplineThicken}
  n3 = Sine(None=ComponentMask(None=TextureCoordinate() {uTiling=0.5, vTiling=0.5}) {g=true})
  n4 = VectorParameter() {parameterName=OpacityGradientRefine, defaultValue=<rgba 0 1 4 0.01>}
  MaterialFunctionCall(Texture=@n3, RGB=@n4.RGB, A=@n4.RGB) {materialFunction=&MF_FX_Refine}
  n5 = DynamicParameter() {
   parameterIndex=1
   defaultValue=<rgba 0 0 0 0.5>
   paramNames [ThicknessMultiplier/NoiseDistortion, UV.x Index/Speed Y, UV.y Position/NoiseErosion, BoltShaderThickness/NoiseDistortion]
  }
  MaterialFunctionCall(G=@n5.UV.x Index/Speed Y) {materialFunction=&AppendMany}
  n6 = ParticleColor()
  n7 = ScalarParameter() {defaultValue=0.35, parameterName=ThinBoltCore}
  n8 = NamedRerouteUsage()
  n9 = SmoothStep(Min=@n7, Max=Add(A=@n7) {constB=0.1}, Value=@n8) {constMax=0.55, constMin=0.5}
  n10 = Multiply(A=@n6.RGB, B=Constant3Vector() {constant=<rgba 5 5 5 0>})
  MaterialFunctionCall(Emissive=Add(A=@n6.RGB, B=Multiply(A=@n9, B=@n10))) {materialFunction=&MF_FX_EmissiveAdjustments}
  Multiply() {constA=1}
  n11 = TextureCoordinate() {uTiling=0.5, vTiling=0.5}
  n12 = VectorParameter() {parameterName=GradientRefine, defaultValue=<rgba 0.1 1 1 0.5>}
  n13 = MaterialFunctionCall(Texture=Sine(None=ComponentMask(None=@n11) {g=true}), RGB=@n12.RGB, A=@n12.RGB) {materialFunction=&MF_FX_Refine}
  n14 = OneMinus(None=ComponentMask(None=TextureCoordinate()) {g=true})
  n15 = DynamicParameter() {
   parameterIndex=1
   defaultValue=<rgba 0 0 0 0.5>
   paramNames [ThicknessMultiplier/NoiseDistortion, UV.x Index/Speed Y, UV.y Position/NoiseErosion, BoltShaderThickness/NoiseDistortion]
  }
  n16 = VectorParameter() {parameterName=NoiseScale, defaultValue=<rgba 0.5 0.5 0 1>}
  n17 = ComponentMask(None=@n16.RGB) {g=true, r=true}
  n18 = Panner(Coordinate=Multiply(A=TextureCoordinate(), B=Multiply(A=@n17, B=@n16.B))) {bFractionalPart=true, speedY=1}
  n19 = Subtract(A=Multiply(A=TextureSample(UVs=@n18) {texture=/BRCosmetics/Athena/Effects/Particles/Skydiving/NS_Hightower_Date/T_AtmosphericCloudNoise_Vertical}.R) {constB=2})
  n20 = MaterialFunctionCall(
   UVs=TextureCoordinate()
   Texture Scale=ComponentMask(None=VectorParameter() {group=UV, parameterName=Tiling, defaultValue=<rgba 6 1 0 1>}.RGB) {g=true, r=true}
  ) {materialFunction=&ScaleUVsByCenter}
  n21 = DynamicParameter() {defaultValue=<rgba 1 0 0 0.5>, paramNames [ThicknessMultiplier, UV.x Index, UV.y Position, BoltShaderThickness]}
  n22 = Constant() {r=20}
  n23 = Multiply(A=Time(), B=@n22) {constB=4}
  n24 = Add(A=@n21.UV.y Position, B=Divide(A=Ceil(None=@n23), B=@n22))
  n25 = TextureSample(UVs=Add(A=Multiply(A=@n15.ThicknessMultiplier/NoiseDistortion, B=@n19), B=Panner(Coordinate=@n20.UVs, Time=@n24) {bFractionalPart=true, speedY=1.1})) {
   samplerType=SAMPLERTYPE_LinearGrayscale
   texture=&/T_FX_Snapfreeze_Bolts
  }
  n26 = Multiply(A=Saturate(None=Add(A=@n14, B=@n15.UV.y Position/NoiseErosion)), B=@n25.R)
  Multiply(A=@n13.Result, B=@n26)
  n27 = ComponentMask(None=TextureCoordinate() {uTiling=0.5, vTiling=0.5}) {g=true}
  n28 = VectorParameter() {parameterName=ColorRefine, defaultValue=<rgba 0.5 1 1 0>}
  Multiply(
   A=@n26
   B=MaterialFunctionCall(Texture=Sine(None=@n27), RGB=@n28.RGB, A=@n28.RGB) {materialFunction=&MF_FX_Refine}.Result
  )
  n29 = Normalize(VectorInput=Subtract(A=CameraPositionWS(), B=WorldPosition().XYZ))
  n30 = DynamicParameter() {parameterIndex=3, defaultValue=<rgba 1 0 0 0.5>, paramNames [ThicknessMultiplier, UV.x Index, UV.y Position, BoltShaderThickness]}
  n31 = MaterialFunctionCall(R=@n30.ThicknessMultiplier, G=@n30.ThicknessMultiplier, B=@n30.ThicknessMultiplier) {materialFunction=&AppendMany}
  n32 = OneMinus(None=SphereMask(
    A=MaterialFunctionCall(Plane Normal Axis=@n29, Plane Center=@n31.RGB) {materialFunction=&VirtualPlaneCoordinates}.World Position - Centered
    B=Constant()
    Radius=@n30.BoltShaderThickness
   ) {hardnessPercent=50})
  n33 = Multiply(A=Power(Base=@n8, Exp=ScalarParameter() {defaultValue=2, parameterName=Exp})) {constB=0.125}
  n34 = TextureCoordinate()
  n35 = ComponentMask(None=@n34) {g=true}
  n36 = DynamicParameter() {defaultValue=<rgba 1 0 0 0.5>, paramNames [ThicknessMultiplier, UV.x Index, UV.y Position, BoltShaderThickness]}
  n37 = Add(A=@n36.BoltShaderThickness, B=ScalarParameter() {defaultValue=0.1, group=Shape Control, parameterName=BoltSharpness}) {constB=0.2}
  n38 = Multiply(
   A=Add(A=Multiply(A=@n33, B=@n35), B=SmoothStep(Min=@n36.BoltShaderThickness, Max=@n37, Value=@n8) {constMax=0.55, constMin=0.5})
   B=SphereMask(A=ComponentMask(None=@n34) {r=true}, B=Constant() {r=0.5}, Radius=ScalarParameter() {defaultValue=0.5, parameterName=Radius}) {attenuationRadius=0.5, hardnessPercent=0}
  )
  n39 = Constant() {r=400}
  Multiply(
   A=Multiply(
    A=Multiply(A=@n32, B=@n38)
    B=MaterialFunctionCall(Distance=@n39) {materialFunction=&Near_Camera_Fade}.Result
   )
   B=SmoothStep(Min=ScalarParameter() {parameterName=Param_1}, Max=ScalarParameter() {defaultValue=1, parameterName=Param_2}, Value=@n3)
  )
  NamedRerouteDeclaration(None=@n26) {name=LightningSDF, nodeColor=<rgba 0.8047 1 0 1>}
  NamedRerouteDeclaration(None=@n13.Result) {name=LeadingEdgeMask, nodeColor=<rgba 1 0 0.8437 1>}
  Saturate(None=ComponentMask(None=TextureCoordinate() {vTiling=2}) {g=true})
  n40 = ComponentMask(None=TextureCoordinate()) {g=true}
  Saturate(None=Multiply(A=OneMinus(None=@n40)) {constB=16})
  ScalarParameter()
  ScalarParameter() {defaultValue=1, parameterName=MasterOpacity}
  n41 = DynamicParameter() {
   parameterIndex=2
   defaultValue=<rgba 1 0 0 0.5>
   paramNames [ThicknessMultiplier/???/ErosionBarPos, UV.x Index/Speed Y/ErosionBarWidth, UV.y Position/NoiseErosion, BoltShaderThickness]
  }
  SphereMask(A=@n41.ThicknessMultiplier/???/ErosionBarPos, B=@n14, Radius=@n41.ThicknessMultiplier/???/ErosionBarPos) {hardnessPercent=50}
 ]
}
