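# Niagara system for the Snapfreeze skydiving contrail cosmetic, drawing a purple plasma-globe effect around the character's skeletal mesh.
# EndpointsV2 builds ribbon arcs from bones to an energy sphere's radius, with mesh and sprite contact points, alongside hand arcs, flickering glows and particle bursts.
# A system timeline and waveform drive intensity and arc rate.
# Glide lean steers arc direction, while front-end, platform and time-of-day modifiers tune the look.
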NiagaraSystem /BRCosmetics/Athena/Effects/Particles/Skydiving/NS_Contrail_Snapfreeze/NS_Snapfreeze_Contrail
dependencies {
 usedDataInterfaces [
  &NiagaraDataInterfaceMeshRendererInfo
  &NiagaraDataInterfaceCurve
  &NiagaraDataInterfaceArrayFloat
  &NiagaraDataInterfaceVector2DCurve
  &NiagaraDataInterfaceColorCurve
  &NiagaraDataInterfacePlatformSet
  &NiagaraDataInterfaceArrayColor
  &NiagaraDataInterfaceArrayFloat3
  &NiagaraDataInterfaceCamera
  &NiagaraDataInterfaceSkeletalMesh
  &NiagaraDataInterfaceArrayPosition
 ]
 usedDynamicInputs [&SetBoolByFloatComparison]
 usedModules [
  &self:SystemSpawnScript.SystemScriptSource.SystemScriptGraph.NiagaraNodeAssignment_0.SetVariables_#1
  &self:SystemSpawnScript.SystemScriptSource.SystemScriptGraph.NiagaraNodeAssignment_3.SetVariables_#2
  &self:SystemSpawnScript.SystemScriptSource.SystemScriptGraph.NiagaraNodeAssignment_1.SetVariables_#3
  &self:SystemSpawnScript.SystemScriptSource.SystemScriptGraph.NiagaraNodeAssignment_2.SetVariables_#4
  &self:SystemSpawnScript.SystemScriptSource.SystemScriptGraph.NiagaraNodeAssignment_5.SetVariables_#5
  &self:SystemSpawnScript.SystemScriptSource.SystemScriptGraph.NiagaraNodeAssignment_4.SetVariables_#6
  &SystemState
  &Timeline
  /Niagara/Noise/Waveform
  &SampleSkeletalMesh
 ]
}
emitters {
 Decal_ {
  emitterUpdateScript [
   &EmitterState(
    Inactive Response=::NewEnumerator1
    Life Cycle Mode=::NewEnumerator1
    Loop Behavior=::NewEnumerator1
    Loop Duration Mode=::NewEnumerator1
    UseLoopDelay=false
   )
   &SpawnBurst_Instantaneous(Spawn Count=12, Use Spawn Probability=false, UseLoopCountLimit=false)
   /SetVariables(Emitter.Flicker:NiagaraFloat=&Add_Float(A=&Multiply_Float(A=$System.Waveform.Waveform, B=0.5), B=0.5)) {isSetParameters=true}
  ]
  particleSpawnScript [
   &PartitionParticles(Number Of Partitions=3)
   &FN_SetSpawnGroup()
   &InitializeParticle(
    Color=&MergeHelper_ColorSwitch(
     0=<rgba 0.8026 0.3851 1 1>
     1=<rgba 72.9641 0 800 1>
     2=<rgba 0.7687 0.1776 2 1>
     3=?
     Index=$Particles.PartitionParticles.Partition
     Num=::NewEnumerator1
    )
    Lifetime=0.1
    Sprite Size Mode=::NewEnumerator3
    Uniform Sprite Size=&MergeHelper_FloatSwitch(0=80, 1=120, 2=80, 3=?, Index=$Particles.PartitionParticles.Partition, Num=::NewEnumerator1)
    UsePositionOffset=false
    Write Material Random=true
   )
   &CameraOffset(Camera Offset Amount=30)
   &DynamicMaterialParameters(
    Index 0 Param 1=0.2
    Index 0 Param 2=1
    Module.Param0WriteEnabled=true
    Module.Param1WriteEnabled=true
    Module.Param2WriteEnabled=true
    Module.Param3WriteEnabled=true
   )
  ]
  particleUpdateScript [
   &ParticleState(Kill Particles When Lifetime Has Elapsed=false, Let Infinitely Lived Particles Die When Emitter Deactivates=true)
   &SkeletalMeshLocation(
    Bone / Socket Index=$Particles.SpawnGroup
    Bone Sampling Mode=::NewEnumerator4
    MaskSpawnGroup=false
    Sampled Position Offset=&MergeHelper_VectorSwitch(0=<0 0 0>, 1=<0 0 10>, 2=<0 0 0>, 3=<0 0 0>, 4=<0 0 0>, Index=$Particles.PartitionParticles.Partition, Num=::NewEnumerator3)
    Skeletal Mesh=$System.Skeletal Mesh
   )
   &ScaleColor(
    Scale Alpha=&Lerp_Float(A=0.2, B=0.5, Alpha=$System.Waveform.Waveform)
    Scale RGB=&VectorFromFloat(Value=1)
    ScaleA=true
    ScaleRGB=true
   )
   &FN_Light_Attributes(
    Color=$Particles.Color
    Create Color Attribute=true
    Create Exponent Attribute=true
    Create Radius Attribute=true
    Exponent=5
    Radius=35
    Scale Color=&Multiply_Float(A=1, B=$Emitter.Flicker)
   )
  ]
  rendererClasses [&NiagaraSpriteRendererProperties]
  renderers [
   {rendererClass=&NiagaraSpriteRendererProperties}
   {rendererClass=&NiagaraSpriteRendererProperties, rendererIndex=1}
  ]
 }
 Dirt {
  emitterUpdateScript [
   &EmitterState(Inactive Response=::NewEnumerator1, Life Cycle Mode=::NewEnumerator1, Loop Behavior=::NewEnumerator1, Loop Duration=1, UseLoopDelay=false)
   &SpawnBurst_Instantaneous(Spawn Count=30, Use Spawn Probability=false, UseLoopCountLimit=false)
  ]
  particleSpawnScript [
   &InitializeParticle(
    Color Maximum=<rgba 3.0659 0.3125 10 1>
    Color Minimum=<rgba 0.2922 0.0298 0.9531 1>
    Color Mode=::NewEnumerator2
    Lifetime Max=1.5
    Lifetime Min=0.5
    Lifetime Mode=::NewEnumerator1
    Sprite Size Mode=::NewEnumerator1
    Uniform Sprite Size Max=15
    Uniform Sprite Size Min=6
    UsePositionOffset=true
    Write Material Random=true
   )
   &ShapeLocation(
    Large Radius=70
    MaskSpawnGroup=false
    Non Uniform Scale=<1 1 0.5>
    Offset=<0 0 80>
    Offset Mode=::NewEnumerator0
    Override Seed=false
    Shape Primitive=::NewEnumerator3
    UseOwnerScale=true
   )
   &AddVelocity(
    Module.Use Falloff Distance=false
    Origin Offset=<0 0 60>
    Override Seed=false
    Velocity Mode=::NewEnumerator1
    Velocity Speed=&RandomRangeFloat(Minimum=300, Maximum=600)
   )
   &AddVelocity() {module=AddVelocity001}
  ]
  particleUpdateScript [
   &ParticleState()
   &ScaleSpriteSize(Scale Factor=<0.4 1>, Scale Sprite Size Mode=::NewEnumerator2) {module=ScaleSpriteSize001}
   &ScaleSpriteSize(Uniform Curve Sprite Scale=DataInterface() {propertyValues {Curve {keys [
        {arriveTangent=8.2127, interpMode=RCIM_Cubic, leaveTangent=8.2127, tangentMode=RCTM_User, time=0, value=0.5}
        {
         arriveTangent=4.2448
         arriveTangentWeight=0.232
         interpMode=RCIM_Cubic
         leaveTangent=4.2448
         leaveTangentWeight=2.2538
         tangentMode=RCTM_User
         time=0.07
         value=1
        }
        {arriveTangent=-2.5194, interpMode=RCIM_Cubic, leaveTangent=-2.5194, tangentMode=RCTM_User, time=0.75, value=0.8}
        {arriveTangent=-4.4982, interpMode=RCIM_Cubic, leaveTangent=-4.4982, tangentMode=RCTM_User, time=1, value=0}
       ]}}})
   &DynamicMaterialParameters(
    Index 0 Param 1=&FloatFromCurve(
     FloatCurve=DataInterface() {propertyValues {Curve {keys [
         {arriveTangent=-5.8452, interpMode=RCIM_Cubic, time=0, value=1}
         {arriveTangentWeight=0.0434, interpMode=RCIM_Cubic, leaveTangentWeight=0.217, tangentMode=RCTM_User, time=0.2, value=0}
         {arriveTangent=3.4164, interpMode=RCIM_Cubic, leaveTangent=3.4164, tangentMode=RCTM_User, time=0.8, value=0.7}
         <at 1 0 RCIM_Cubic>
        ]}}}
     Scale Curve=3
    )
    Index 0 Param 3=1
    Index 0 Param 4=&FloatFromCurve(
     FloatCurve=DataInterface() {propertyValues {Curve {keys [
         {interpMode=RCIM_Cubic, leaveTangent=-1.5195, tangentMode=RCTM_Break, time=0, value=1}
         {arriveTangent=0.0067, interpMode=RCIM_Cubic, tangentMode=RCTM_Break, time=1, value=0}
        ]}}}
     Scale Curve=&RandomRangeFloat(Minimum=10, Maximum=30)
    )
    Module.Param0WriteEnabled=true
    Module.Param1WriteEnabled=true
    Module.Param2WriteEnabled=true
    Module.Param3WriteEnabled=true
   )
   &Drag(Drag=&RandomRangeFloat(Minimum=1, Maximum=2), Use Linear Drag=true, Use Rotational Drag=true)
   &GravityForce(Gravity=&RandomRangeVector(Minimum=<0 0 -200>, Maximum=<0 0 -400>))
   &SolveForcesAndVelocity(Clamp Velocity=false, Limit Acceleration=false)
   &ScaleColor(
    Scale Alpha=&Lerp_Float(
     A=0
     B=1
     Alpha=&FloatFromCurve(
      FloatCurve=DataInterface() {propertyValues {Curve {keys [
          {interpMode=RCIM_Cubic, tangentMode=RCTM_User, time=0, value=1}
          {interpMode=RCIM_Cubic, tangentMode=RCTM_User, time=0.5, value=1}
          {tangentMode=RCTM_User, time=1, value=0}
         ]}}}
      Scale Curve=&Multiply_Float(A=1, B=&FNTimeOfDayScalar(Night Scale=0.6, Morning Scale=0.9, Evening Scale=0.9))
     )
    )
    Scale RGB=&VectorFromFloat(Value=&FloatFromCurve(
      FloatCurve=DataInterface() {propertyValues {Curve {keys [
          {arriveTangent=-2.3805, interpMode=RCIM_Cubic, leaveTangent=-2.3805, tangentMode=RCTM_User, time=0, value=1}
          {interpMode=RCIM_Cubic, tangentMode=RCTM_User, time=1, value=0}
         ]}}}
      Scale Curve=&RandomRangeFloat(Minimum=1, Maximum=1.2)
     ))
    ScaleA=true
    ScaleRGB=true
   )
  ]
  rendererClasses [&NiagaraSpriteRendererProperties]
  renderers [{rendererClass=&NiagaraSpriteRendererProperties}]
 }
 EndpointsV2 {
  emitterSpawnScript [/SetVariables(Emitter.MaxPartitionAmount=100) {isSetParameters=true}]
  emitterUpdateScript [
   &EmitterState(
    Life Cycle Mode=::NewEnumerator1
    Loop Delay=&RandomRangeFloat(Minimum=0, Maximum=0)
    Loop Duration=&Lerp_Float(A=&RandomRangeFloat(Minimum=0.25, Maximum=0.1), B=0.05, Alpha=$System.Timeline.Time)
    Recalculate Duration Each Loop=true
    UseLoopDelay=false
   )
   &SpawnBurst_Instantaneous(
    Spawn Count=$System.BeamSubdivision
    Spawn Probability=&Lerp_Float(A=1, B=0.2, Alpha=$System.Timeline.Time)
    Use Spawn Probability=true
    UseLoopCountLimit=false
   )
   &SpawnBurst_Instantaneous(
    Spawn Count=$System.BeamSubdivision
    Spawn Group=1
    Spawn Probability=&Lerp_Float(A=1, B=0.2, Alpha=$System.Timeline.Time)
    Use Spawn Probability=true
    UseLoopCountLimit=false
   ) {module=SpawnBurst_Instantaneous001}
  ]
  particleSpawnScript [
   &PartitionParticles(
    Normalized Index In Partition Includes One=true
    Number Of Partitions=$Emitter.MaxPartitionAmount
    Particles Per Partition:NiagaraInt32=$System.BeamSubdivision
    Partition Mode:ENiagara_PartitionMode=::NewEnumerator1
    Write Partition to Ribbon ID=::NewEnumerator0
   )
   &PartitionParticles(
    ID to Partition:NiagaraInt32=$Particles.PartitionParticles.Partition
    Particles Per Partition=2
    Partition Mode:ENiagara_PartitionMode=::NewEnumerator1
   ) {module=PartitionParticles001}
   &PartitionParticles(
    Particles Per Partition:NiagaraInt32=&Multiply_Int(A=$System.BeamSubdivision, B=2)
    Partition Count Mode:ENiagara_InfiniteLoopDuration=::NewEnumerator1
    Partition Mode:ENiagara_PartitionMode=::NewEnumerator1
   ) {module=PartitionParticles002}
   &PartitionParticles(
    A:NiagaraFloat=&MakeFloatFromInt(Int=$Particles.PartitionParticles002.Index In Partition)
    B:NiagaraFloat=&MakeFloatFromInt(Int=6)
    Comparison Type:ENiagaraBooleanLogicOps_v2=::NewEnumerator4
    Partition Mode:ENiagara_PartitionMode=::NewEnumerator4
   ) {module=PartitionParticles004}
   /SetVariables(Particles.PlasmaInstability:NiagaraFloat=&FloatFromCurve(
     FloatCurve=DataInterface() {propertyValues {Curve {keys [<at 0 0 RCIM_Constant>,<at 0.5 0.4 RCIM_Constant>,<at 0.8 1 RCIM_Constant>]}}}
     CurveIndex=&RandomRangeFloat(Minimum=0, Maximum=1, Randomness Mode=::NewEnumerator1, Random Seed=$Emitter.LoopCount, Fixed Random Seed=true)
    )) {isSetParameters=true}
   /SetVariables(Particles.PlasmaFactor:NiagaraFloat=&FloatFromCurve(
     FloatCurve=DataInterface() {propertyValues {Curve {keys [<at 0.4 1.5 RCIM_Constant>,<at 0.4459 1.0363 RCIM_Constant>,<at 0.5575 0.7644 RCIM_Constant>,<at 0.6 0.5 RCIM_Constant>,<at 0.7 0.2 RCIM_Constant>]}}}
     CurveIndex=$Particles.PlasmaInstability
    )) {isSetParameters=true}
   &InitializeParticle(
    Color=&MultiplyLinearColor(
     Linear Color=&SelectLinearColorFromArray(
      Array Sampling Mode=::NewEnumerator1
      Direct Array Index=$Particles.PartitionParticles.Partition
      Direct Array Mode=::NewEnumerator1
      Color Selection Array=DataInterface() {propertyValues {ColorData [
         <rgba 0.2 0.2 0.2 1>
         <rgba 0.4 0.4 0.4 1>
         <rgba 0.1 0.1 0.1 1>
         <rgba 0.1 0.1 0.1 1>
         <rgba 0.1 0.1 0.1 1>
         <rgba 1 1 1 1>
         <rgba 0.1 0.1 0.1 1>
        ]}}
      Random Seed=?
     )
     Scale Factor=&MakeVector4_FromLinearColor(LinearColor=<rgba 0.7832 0 4 1>)
    )
    Lifetime=&Multiply_Float(
     A=&Lerp_Float(A=1, B=0.4, Alpha=$Particles.PlasmaInstability)
     B=&Lerp_Float(A=1, B=0.5, Alpha=$System.Timeline.Time)
    )
    Mesh Scale Mode=::NewEnumerator3
    Mesh Uniform Scale=0.25
    Position Mode=::NewEnumerator0
    Ribbon Width=15
    Ribbon Width Mode=::NewEnumerator1
    Write Material Random=true
   )
   &InitialMeshOrientation(Mesh Orientation Mode=::NewEnumerator2, Orientation Coordinate Space=::Local, Orientation Vector=<0 0 0>, Use Rotation Vector=true)
  ]
  particleUpdateScript [
   &ParticleState()
   &SkeletalMeshLocation(
    Bone / Socket Index=&ModuloInt(A=$Particles.PartitionParticles.Partition, B=4)
    Bone Sampling Mode=::NewEnumerator4
    MaskSpawnGroup=false
    Orientation Sampling=::NewEnumerator1
    Position Sampling=::NewEnumerator2
    Skeletal Mesh=$System.Skeletal Mesh
    Use Sampled Mesh Attribute as Ribbon ID=false
   )
   &ScaleMeshSize(Scale Factor=&VectorFromFloat(Value=&MakeCustomFloatFromBool(
      Bool=&SetBoolByIntComparison(
       A=$Particles.PartitionParticles.Index In Partition
       B=/Niagara/DynamicInputs/Subtract/Subtract_Integer(A:NiagaraInt32=$System.BeamSubdivision, B:NiagaraInt32=1)
       Comparison Type=::NewEnumerator4
      )
      True Float=&FloatFromCurve(
       FloatCurve=DataInterface() {propertyValues {Curve {keys [<at 0 0.75 RCIM_Constant>,<at 0.4 0.9 RCIM_Constant>,<at 0.8 1.4 RCIM_Constant>]}}}
       CurveIndex=$Particles.PlasmaInstability
      )
      False Float=0
     )))
   &ScaleColor(
    Scale Alpha=&FloatFromCurve(FloatCurve=DataInterface() {propertyValues {Curve {keys [
         {interpMode=RCIM_Cubic, tangentMode=RCTM_User, time=0, value=0}
         {arriveTangent=1.4286, interpMode=RCIM_Cubic, leaveTangent=1.4286, tangentMode=RCTM_User, time=0.15, value=1}
         {interpMode=RCIM_Cubic, tangentMode=RCTM_User, time=1, value=0}
        ]}}})
    Scale RGB=&VectorFromFloat(Value=&Multiply_Float(
      A=&Multiply_Float(
       A=&FloatFromCurve(
        FloatCurve=DataInterface() {propertyValues {Curve {keys [<at 0 0.9 RCIM_Constant>,<at 0.4 1.4 RCIM_Constant>,<at 0.8 2 RCIM_Constant>]}}}
        CurveIndex=$Particles.PlasmaInstability
       )
       B=&FloatFromCurve(
        FloatCurve=DataInterface() {propertyValues {Curve {keys [<at 0 10>,<at 0.2 1>,<at 1 0>]}}}
        Scale Curve=&Lerp_Float(A=1, B=0.5, Alpha=$System.Timeline.NormalizedTime)
       )
      )
      B=&Lerp_Float(A=1, B=0.2, Alpha=$System.Waveform.Waveform)
     ))
    ScaleA=true
    ScaleRGB=true
   )
   /SetVariables(Particles.ParticleDirectionVector:Vector3f=&Multiply_Vector(
     A=&MakeVector(
      X=&Lerp_Float(
       A=1
       B=-1
       Alpha=&MakeFloatFromInt(Int=&ModuloInt(A=$Particles.PartitionParticles.Partition, B=2))
      )
      Y=1
      Z=1
     )
     B=&Lerp_Vector(
      A=&RandomRangeVector(
       Minimum=<1 0 0>
       Maximum=<1 2 0>
       Randomness Mode=::NewEnumerator1
       Random Seed=$Particles.PartitionParticles.Partition
       Fixed Random Seed=true
      )
      B=&Lerp_Vector(
       A=&RandomRangeVector(
        Minimum=<0.2 -0.5 0>
        Maximum=<1 -1 0>
        Randomness Mode=::NewEnumerator1
        Random Seed=$Particles.PartitionParticles.Partition
        Fixed Random Seed=true
       )
       B=&FN_FrontEndVector3Modifier(
        Gameplay=&RandomRangeVector(
         Minimum=<1 0.2 1>
         Maximum=<1 0.8 1>
         Randomness Mode=::NewEnumerator1
         Random Seed=$Particles.PartitionParticles.Partition
         Fixed Random Seed=true
        )
        Front End=&RandomRangeVector(
         Minimum=<1 -0.5 0.1>
         Maximum=<1 -1 0.1>
         Randomness Mode=::NewEnumerator1
         Random Seed=$Particles.PartitionParticles.Partition
         Fixed Random Seed=true
        )
       )
       Alpha=$User.ParaGlideLeanAlpha
      )
      Alpha=&MakeFloatFromInt(Int=$Particles.PartitionParticles001.Partition)
     )
    )) {isSetParameters=true}
   /SetVariables(Particles.EndpointPosition:NiagaraPosition=&AddVectorToPosition(
     Position=$System.SampleSkeletalMesh.SampledPosition
     Vector=&RotateVectorByQuaternion(
      Vector To Rotate=&MakeVector(X=$System.SphereRadius, Y=0, Z=0)
      Quaternion=&OrientMeshToVector(
       Look At Direction=&AddVectorToPosition(
        Position=Linked() {linkedVariable {name=Particles.ParticleDirectionVector, type=Vector3f}}
        Vector=&Multiply_Vector(
         A=&VectorFromFloat(Value=&Multiply_Float(
           A=&FloatFromCurve(
            FloatCurve=DataInterface() {propertyValues {Curve {keys [{tangentMode=RCTM_User, time=0, value=0}, {tangentMode=RCTM_User, time=1, value=1}]}}}
            Scale Curve=&Multiply_Float(
             A=&Lerp_Float(
              A=1
              B=&RandomRangeFloat(Minimum=0.15, Maximum=0.3, Randomness Mode=::NewEnumerator1, Random Seed=$Particles.PartitionParticles.Partition)
              Alpha=$Particles.PlasmaInstability
             )
             B=1
            )
           )
           B=1
          ))
         B=&Lerp_Vector(
          A=<0 0 1.5>
          B=&FN_FrontEndVector3Modifier(Gameplay=<0 1 0>, Front End=<0 0 1.5>)
          Alpha=$System.Timeline.Time
         )
        )
       )
       Reference Axis Vector=<1 0 0>
      )
     )
    )) {isSetParameters=true}
   /SetVariables(
    Particles.Position=&Lerp_Vector(
     A=$Particles.SkeletalMeshLocation.SampledPosition
     B=Linked() {linkedVariable {name=Particles.EndpointPosition, type=NiagaraPosition}}
     Alpha=$Particles.PartitionParticles.Normalized Index In Partition
    )
    Particles.RibbonWidth=&Multiply_Float(
     A=&FloatFromCurve(
      FloatCurve=DataInterface() {propertyValues {Curve {keys [<at 0 0.5>, <at 0.1 1>, {leaveTangent=-1.5932, time=1, value=0.5}]}}}
      CurveIndex=$Particles.Initial.PartitionParticles.Normalized Index In Partition
     )
     B=&Multiply_Float(
      A=$Particles.Initial.RibbonWidth
      B=&FloatFromCurve(FloatCurve=DataInterface() {propertyValues {Curve {keys [<at 0 1>,<at 1 0.7>]}}})
     )
    )
   ) {isSetParameters=true}
   /SetVariables(
    Particles.ContactPointColor:LinearColor=&MultiplyLinearColor(
     Linear Color=$Particles.Color
     Scale Factor=&MakeVector4_FromLinearColor(LinearColor=<rgba 10 0.9903 2.5911 1>)
    )
    Particles.ContactPointDetailColor:LinearColor=&MultiplyLinearColor(
     Linear Color=$Particles.Color
     Scale Factor=&MakeVector4_FromLinearColor(LinearColor=<rgba 40 3.9613 10.3643 1>)
    )
    Particles.ContactPointDetailScale:Vector3f=&Multiply_VectorByFloat(Vector=$Particles.Scale, Float=1.7)
   ) {isSetParameters=true}
   &FN_Sprite_Attributes(
    Color=&MultiplyLinearColor(Linear Color=<rgba 3.2878 0 4 1>, Scale Factor=&MakeVector4_FromLinearColor(LinearColor=$Particles.Color))
    Create Color Attribute=true
    Create SpriteSize Attribute=true
    SpriteSize=&Vector2DFromFloat(Value=&MakeCustomFloatFromBool(
      Bool=&SetBoolByIntComparison(
       A=$Particles.PartitionParticles.Index In Partition
       B=&Subtract_Integer(A:NiagaraInt32=$System.BeamSubdivision, B:NiagaraInt32=1)
       Comparison Type=::NewEnumerator4
      )
      True Float=50
      False Float=0
     ))
    Write to Module Namespace=true
   ) {module=FN_Sprite_Attributes001}
   &FN_Sprite_Attributes(
    Camera Offset=20
    Color=&MultiplyLinearColor(
     Linear Color=&HueShiftLinearColor(Hue=0, Saturation=0, Value=0.5, Linear Color=$Particles.ContactPointDetailColor)
     Scale Factor=&MakeVector4_FromLinearColor(LinearColor=&MultiplyLinearColor(
       Linear Color=<rgba 0.1677 0 1 1>
       Scale Factor=&MakeFloat4FromVectorAndFloat(
        "Vector (RGB)"=<1 1 1>
        "Float (Alpha)"=&FloatFromCurve(
         FloatCurve=DataInterface() {propertyValues {Curve {keys [{interpMode=RCIM_Cubic, tangentMode=RCTM_User, time=0, value=1}, {interpMode=RCIM_Cubic, tangentMode=RCTM_User, time=0.4, value=0}]}}}
         CurveIndex=&Divide_Float(A=$Particles.NormalizedAge, B=&RandomRangeFloat(Minimum=1, Maximum=2))
        )
       )
      ))
    )
    Create Camera Offset Attribute=true
    Create Color Attribute=true
    Create SpriteSize Attribute=true
    SpriteSize=&Vector2DFromFloat(Value=&MakeCustomFloatFromBool(
      Bool=&SetBoolByIntComparison(A=$Particles.PartitionParticles.Index In Partition, B=0, Comparison Type=::NewEnumerator4)
      True Float=50
      False Float=0
     ))
    Write to Module Namespace=true
   )
   &UpdateMeshOrientation(
    Facing Interpolation=true
    Facing Interpolation Rate=1000
    Facing Origin Coordinate Space:ENiagaraCoordinateSpace=::Local
    Facing Position:NiagaraPosition=$System.SampleSkeletalMesh.SampledPosition
    Facing Position Coordinate Space:ENiagaraCoordinateSpace=::Local
    Orientation Method=::NewEnumerator2
    Override Quat=false
    Side Coordinate Space:ENiagaraCoordinateSpace=::Local
    Side Direction:Vector3f=<0 0 0>
    Stabilization Mode:ENiagara_OrientVectorCount=::NewEnumerator2
    TransformFacingOrigin=true
    TransformFacingPosition=true
    TransformSide=true
   )
   /SetVariables(Particles.Position=&Lerp_Vector(
     A=Linked() {linkedVariable {name=Particles.Position, type=NiagaraPosition}}
     B=&AddVectorToPosition(
      Position=$Particles.Position
      Vector=&RotateVectorByQuaternion(
       Vector To Rotate=&RandomVector(Vector Scale=10)
       Quaternion=&OrientMeshToVector(Look At Direction=$Particles.ParticleDirectionVector, Reference Axis Vector=<1 0 0>)
      )
     )
     Alpha=&Multiply_Float(
      A=&FloatFromCurve(FloatCurve=DataInterface() {propertyValues {Curve {keys [<at 0 0>,<at 1 1>]}}}, Scale Curve=2)
      B=&FloatFromCurve(
       FloatCurve=DataInterface() {propertyValues {Curve {keys [
           {interpMode=RCIM_Cubic, tangentMode=RCTM_User, time=0.3, value=0}
           {interpMode=RCIM_Cubic, tangentMode=RCTM_User, time=0.5, value=1}
           {interpMode=RCIM_Cubic, tangentMode=RCTM_User, time=0.8, value=0}
          ]}}}
       CurveIndex=$Particles.PartitionParticles.Normalized Index In Partition
      )
     )
    )) {isSetParameters=true}
   &DynamicMaterialParameters(
    Index 0 Param 3=40000
    Module.Index0Param0Write=true
    Module.Index0Param1Write=true
    Module.Index0Param2Write=true
    Module.Index0Param3Write=true
   )
   /SetVariables(
    Particles.ArcColor:LinearColor=&MultiplyLinearColor(
     Linear Color=&MultiplyLinearColor(
      Linear Color=&RandomRangeLinearColor(Minimum=<rgba 0.8536 0.8 1 1>, Maximum=<rgba 1.7252 1.6 2 2>)
      Scale Factor=&MakeVector4_FromFloat(Float=0.6)
     )
     Scale Factor=&MakeVector4_FromLinearColor(LinearColor=&LerpLinearColors(
       StartColor=&MakeLinearColorFromVectorAndFloat(
        "Vector (RGB)"=&Lerp_Vector(
         A=&VectorFromFloat(Value=&FloatFromCurve(FloatCurve=DataInterface() {propertyValues {Curve {keys [<at 0.2771 1>,<at 0.3786 1>]}}}, Scale Curve=20))
         B=&VectorFromFloat(Value=&RandomRangeFloat(Minimum=0, Maximum=2, Evaluation Type=::NewEnumerator1))
         Alpha=&FloatFromCurve(FloatCurve=DataInterface() {propertyValues {Curve {keys [
              {interpMode=RCIM_Cubic, leaveTangent=2.5294, tangentMode=RCTM_Break, time=0, value=0}
              {arriveTangent=-0.2166, interpMode=RCIM_Cubic, tangentMode=RCTM_Break, time=0.4, value=1}
             ]}}})
        )
        "Float (Alpha)"=&FloatFromCurve(FloatCurve=DataInterface() {propertyValues {Curve {keys [
             {tangentMode=RCTM_User, time=0, value=1}
             {arriveTangent=-1, interpMode=RCIM_Cubic, leaveTangent=-1, time=0.0834, value=1}
             {arriveTangent=-0.2668, interpMode=RCIM_Cubic, leaveTangent=-0.2668, tangentMode=RCTM_User, time=1, value=0}
            ]}}})
       )
       EndColor=$Particles.Color
       LerpFactor=0.25
      ))
    )
    Particles.ArcRotation:Quat4f=&MultiplyQuaternion(
     Quaternion A=$Particles.MeshOrientation
     Quaternion B=&OrientMeshToVector(
      Look At Direction=&RandomRangeVector(Minimum=<0 0.5 1>, Maximum=<0 0 1>)
      Reference Axis Vector=<1 0 0>
     )
    )
    Particles.ArcScale:Vector3f=&Multiply_Vector(
     A=&Multiply_Vector(
      A=&RandomRangeVector(Minimum=<0.12 0.12 0.12>, Maximum=<0.18 0.18 0.18>)
      B=&Multiply_Vector(
       A=&VectorFromFloat(Value=&FloatFromCurve(
         FloatCurve=DataInterface() {propertyValues {Curve {keys [{arriveTangent=1.2917, interpMode=RCIM_Cubic, leaveTangent=1.2917, tangentMode=RCTM_User, time=0, value=0}, <at 0.9978 1 RCIM_Cubic>]}}}
         Scale Curve=&MakeCustomFloatFromBool(Bool=$User.bIsFrontEndPreview, True Float=0.8, False Float=0.8)
        ))
       B=&RandomRangeVector(Minimum=<1 1 1.5>, Maximum=<1 1 2>)
      )
     )
     B=&Lerp_Vector(A=<1 1 1>, B=$Particles.Scale, Alpha=-0.1)
    )
    Particles.DY0:Vector4f=&MakeVector4(X=0, Y=0, Z=0, W=13)
    Particles.DY1:Vector4f=&MakeVector4(
     X=&FloatFromCurve(FloatCurve=DataInterface() {propertyValues {Curve {keys [<at 0 0.5>,<at 1 1>]}}})
     Y=1
     Z=&FloatFromCurve(FloatCurve=DataInterface() {propertyValues {Curve {keys [{interpMode=RCIM_Cubic, tangentMode=RCTM_Break, time=0.5, value=-0.5}, <at 1 0.1>]}}})
     W=1
    )
    Particles.DY2:Vector4f=&MakeVector4(
     X=&Multiply_Float(
      A=&FloatFromCurve(
       FloatCurve=DataInterface() {propertyValues {Curve {keys [{interpMode=RCIM_Cubic, leaveTangent=8.1085, tangentMode=RCTM_Break, time=0.8568, value=1}, {tangentMode=RCTM_User, time=1, value=0}]}}}
       Scale Curve=3
      )
      B=&Divide_Float(
       A=&MakeFloatFromVector(Vector=$Particles.Scale, Channel=::NewEnumerator0)
       B=&RandomRangeFloat(Minimum=10, Maximum=15)
      )
     )
     Y=&FloatFromCurve(
      FloatCurve=DataInterface() {propertyValues {Curve {keys [
          {interpMode=RCIM_Cubic, tangentMode=RCTM_Break, time=0.7, value=0}
          {arriveTangent=-0.0859, interpMode=RCIM_Cubic, tangentMode=RCTM_Break, time=1, value=-1}
         ]}}}
      Scale Curve=50
     )
     Z=/Niagara/DynamicInputs/Helpers/Waveform(
      Global Input Position=&RandomRangeFloat(Minimum=0, Maximum=15)
      Global Phase=$Particles.Age
      "Waveform [1]"=::NewEnumerator12
      "[1] Phase Scale"=2
      "[1] Frequency"=1
      "[1] Amplitude Min/Max"=<0 1>
     )
     W=25
    )
    Particles.DY3:Vector4f=&MakeVector4(
     X=0
     Y=&FloatFromCurve(
      FloatCurve=DataInterface() {propertyValues {Curve {keys [
          {interpMode=RCIM_Cubic, leaveTangent=-0.0025, tangentMode=RCTM_Break, time=0, value=0.5}
          {arriveTangent=-0.0316, interpMode=RCIM_Cubic, tangentMode=RCTM_Break, time=1, value=-0.2}
         ]}}}
      Scale Curve=0
     )
     Z=&FloatFromCurve(FloatCurve=DataInterface() {propertyValues {Curve {keys [
          {interpMode=RCIM_Cubic, leaveTangent=1.1034, tangentMode=RCTM_Break, time=0, value=-1}
          {arriveTangent=2.2399, tangentMode=RCTM_User, time=0.5, value=0.125}
         ]}}})
     W=&FloatFromCurve(FloatCurve=DataInterface() {propertyValues {Curve {keys [
          {interpMode=RCIM_Cubic, leaveTangent=1.8089, tangentMode=RCTM_Break, time=0.3, value=-0.5}
          {arriveTangent=2.0757, interpMode=RCIM_Cubic, tangentMode=RCTM_Break, time=0.9, value=1}
         ]}}})
    )
   ) {isSetParameters=true}
  ]
  rendererClasses [
   &NiagaraRibbonRendererProperties
   &NiagaraMeshRendererProperties
   &NiagaraSpriteRendererProperties
  ]
  renderers [
   {rendererClass=&NiagaraRibbonRendererProperties}
   {rendererClass=&NiagaraMeshRendererProperties, rendererIndex=1}
   {rendererClass=&NiagaraMeshRendererProperties, rendererIndex=2}
   {rendererClass=&NiagaraSpriteRendererProperties, rendererIndex=3}
   {rendererClass=&NiagaraMeshRendererProperties, rendererIndex=4}
   {rendererClass=&NiagaraMeshRendererProperties, rendererIndex=5}
   {rendererClass=&NiagaraSpriteRendererProperties, rendererIndex=6}
  ]
 }
 Energy_Sphere {
  emitterUpdateScript [
   &EmitterState(
    Inactive Response=::NewEnumerator1
    Life Cycle Mode=::NewEnumerator1
    Loop Behavior=::NewEnumerator1
    Loop Duration Mode=::NewEnumerator1
    UseLoopDelay=false
   )
   &SpawnBurst_Instantaneous(Spawn Count=1, Use Spawn Probability=true)
  ]
  particleSpawnScript [
   &InitializeParticle(
    Color=<rgba 1.4965 0.2732 4 1.5>
    Lifetime=0.3
    Mesh Scale Mode=::NewEnumerator3
    Mesh Uniform Scale=&Multiply_Float(A=0.01, B=$System.SphereDiameter)
    UsePositionOffset=true
    Write Material Random=true
   ) {module=InitializeParticle001}
   &CameraOffset()
  ]
  particleUpdateScript [
   &ParticleState(Kill Particles When Lifetime Has Elapsed=false)
   &SkeletalMeshLocation(
    Bone / Socket Index=5
    Bone Sampling Mode=::NewEnumerator4
    MaskSpawnGroup=false
    Orientation Sampling=::NewEnumerator1
    Position Sampling=::NewEnumerator2
    Sampled Position Offset=<0 -20 0>
    Skeletal Mesh=$System.Skeletal Mesh
   )
   /SetVariables(Particles.Position=$System.SampleSkeletalMesh.SampledPosition) {isSetParameters=true}
   &ScaleColor(
    Scale Alpha=&FloatFromCurve(
     FloatCurve=DataInterface() {propertyValues {Curve {keys [{arriveTangent=0.1268, interpMode=RCIM_Cubic, leaveTangent=0.1268, tangentMode=RCTM_User, time=0.3, value=0.5}, <at 1 1 RCIM_Cubic>]}}}
     Scale Curve=&Multiply_Float(A=0.5, B=&FNTimeOfDayScalar(Night Scale=0.7, Morning Scale=0.85, Evening Scale=0.85))
    )
    Scale RGB=&MakeVectorFromLinearColorRGB(Color=&MultiplyLinearColor(
      Linear Color=<rgba 1 1 1 1>
      Scale Factor=&MakeFloat4FromVectorAndFloat(
       "Vector (RGB)"=&VectorFromFloat(Value=&Lerp_Float(A=1, B=1.5, Alpha=$System.Timeline.Time))
       "Float (Alpha)"=1
      )
     ))
    ScaleA=true
    ScaleRGB=true
   )
   /SetVariables(Particles.Color002=&MakeLinearColorFromVectorAndFloat(
     "Vector (RGB)"=&MakeVectorFromLinearColorRGB(Color=<rgba 1 1 1 1>)
     "Float (Alpha)"=&FloatFromCurve(FloatCurve=DataInterface() {propertyValues {Curve {keys [<at 0.5 1>,<at 1 1>]}}}, Scale Curve=1.5)
    )) {isSetParameters=true}
   &ScaleMeshSize(Scale Factor=&VectorFromFloat(Value=&FloatFromCurve(FloatCurve=DataInterface() {propertyValues {Curve {keys [
          {arriveTangent=2.3018, interpMode=RCIM_Cubic, leaveTangent=2.3018, tangentMode=RCTM_User, time=0, value=0.2}
          {arriveTangent=0.1704, interpMode=RCIM_Cubic, leaveTangent=0.1704, tangentMode=RCTM_User, time=1, value=1}
         ]}}})))
   &DynamicMaterialParameters(
    Index 0 Param 3=&FloatFromCurve(FloatCurve=DataInterface() {propertyValues {Curve {keys [<at 0.5 0.5>,<at 1 1.5>]}}})
    Index 0 Param 4=&FloatFromCurve(
     FloatCurve=DataInterface() {propertyValues {Curve {keys [
         {arriveTangent=0.0988, interpMode=RCIM_Cubic, leaveTangent=0.0988, tangentMode=RCTM_User, time=0, value=1}
         {arriveTangent=-0.5267, interpMode=RCIM_Cubic, leaveTangent=-0.5267, tangentMode=RCTM_User, time=1, value=0.15}
        ]}}}
     Scale Curve=30
    )
    Module.Index0Param0Write=true
    Module.Index0Param1Write=true
    Module.Index0Param2Write=true
    Module.Index0Param3Write=true
   )
   /SetVariables(
    Particles.Linear Color=<rgba 0.0198 0 0.0729 1.3>
    Particles.Vector:Vector3f=&Multiply_Vector(A=$Particles.Scale, B=&VectorFromFloat(Value=1.12))
    Particles.Vector 4:Vector4f=&MakeVector4(X=0.8, Y=0, Z=0, W=0)
   ) {isSetParameters=true}
   &FN_Mesh_Attributes(
    Color=&MultiplyLinearColor(
     Linear Color=<rgba 0.3242 0.2044 0.8 1>
     Scale Factor=&MakeFloat4FromVectorAndFloat(
      "Vector (RGB)"=<1 1 1>
      "Float (Alpha)"=&Lerp_Float(
       A=&FloatFromCurve(
        FloatCurve=DataInterface() {propertyValues {Curve {keys [
            {arriveTangent=-34.5605, interpMode=RCIM_Cubic, leaveTangent=-34.5605, tangentMode=RCTM_User, time=0, value=3}
            {arriveTangent=-2.3039, interpMode=RCIM_Cubic, leaveTangent=-2.3039, tangentMode=RCTM_User, time=0.0917, value=0.7188}
            {arriveTangent=-0.7914, interpMode=RCIM_Cubic, leaveTangent=-0.7914, time=0.5, value=0.1}
            {arriveTangent=0.0005, interpMode=RCIM_Cubic, leaveTangent=0.0005, tangentMode=RCTM_User, time=1, value=0}
           ]}}}
        Scale Curve=&Multiply_Float(
         A=&FN_FrontEndFloatModifier(Gameplay=1, Front End=0.6)
         B=&FNTimeOfDayScalar(Night Scale=0.5, Morning Scale=0.7, Evening Scale=0.7)
        )
       )
       B=0.15
       Alpha=1
      )
     )
    )
    Create Color Attribute=true
    Create Scale Attribute=true
    Scale=&VectorFromFloat(Value=&Multiply_Float(
      A=&Multiply_Float(A=&Multiply_Float(A=0.01, B=0.8), B=$System.SphereDiameter)
      B=&FloatFromCurve(
       FloatCurve=DataInterface() {propertyValues {Curve {keys [
           {arriveTangent=7.387, interpMode=RCIM_Cubic, leaveTangent=7.387, tangentMode=RCTM_User, time=0, value=0}
           {arriveTangent=1.65, interpMode=RCIM_Cubic, leaveTangent=1.65, time=0.2, value=1.025}
           {arriveTangent=0.2529, interpMode=RCIM_Cubic, leaveTangent=0.2529, tangentMode=RCTM_User, time=1, value=1.65}
          ]}}}
       Scale Curve=0.81
      )
     ))
    Write to Module Namespace=true
   )
   &FN_Mesh_Attributes(
    Color=&MultiplyLinearColor(
     Linear Color=<rgba 0.125 0.0083 0.5 1>
     Scale Factor=&MakeFloat4FromVectorAndFloat(
      "Vector (RGB)"=<1 1 1>
      "Float (Alpha)"=&Lerp_Float(
       A=&FloatFromCurve(
        FloatCurve=DataInterface() {propertyValues {Curve {keys [
            {arriveTangent=-0.0222, interpMode=RCIM_Cubic, leaveTangent=-0.0222, tangentMode=RCTM_User, time=0.2, value=0}
            {arriveTangent=7.4942, interpMode=RCIM_Cubic, leaveTangent=-1.7347, tangentMode=RCTM_Break, time=0.4, value=1}
            {arriveTangent=-0.6673, interpMode=RCIM_Cubic, leaveTangent=-0.6673, tangentMode=RCTM_User, time=0.6, value=0.375}
            {arriveTangent=-0.8204, interpMode=RCIM_Cubic, leaveTangent=-0.8204, tangentMode=RCTM_User, time=0.8, value=0.23}
            {arriveTangent=-0.2805, interpMode=RCIM_Cubic, leaveTangent=-0.2805, tangentMode=RCTM_User, time=1, value=0}
           ]}}}
        Scale Curve=&Multiply_Float(
         A=&FN_FrontEndFloatModifier(
          Gameplay=&FN_PickFloatDependOnPlatform(TargetPlatformActive=0.6, TargetPlatformNoActive=1)
          Front End=0.5
         )
         B=&FNTimeOfDayScalar(Night Scale=0.45, Morning Scale=0.75, Evening Scale=0.6)
        )
       )
       B=0.4
       Alpha=1
      )
     )
    )
    Create Color Attribute=true
    Create Scale Attribute=true
    Scale=&VectorFromFloat(Value=&Multiply_Float(
      A=&Multiply_Float(A=&Multiply_Float(A=0.8, B=0.01), B=$System.SphereDiameter)
      B=&FloatFromCurve(FloatCurve=DataInterface() {propertyValues {Curve {keys [
           {arriveTangent=2.5615, interpMode=RCIM_Cubic, leaveTangent=2.5615, tangentMode=RCTM_User, time=0, value=0.3}
           {arriveTangent=1.2, interpMode=RCIM_Cubic, leaveTangent=1.2, time=0.3199, value=1.052}
           {arriveTangent=0.1533, interpMode=RCIM_Cubic, leaveTangent=0.1533, tangentMode=RCTM_User, time=1, value=1.5}
          ]}}})
     ))
    Write to Module Namespace=true
   ) {module=FN_Mesh_Attributes001}
   &FN_Light_Attributes(
    Color=<rgba 78.2764 46.3932 300 1>
    Create Color Attribute=true
    Create Exponent Attribute=true
    Create Position Attribute=true
    Create Radius Attribute=true
    Exponent=1
    Position:NiagaraPosition=$Particles.SkeletalMeshLocation.SampledPosition
    Radius=80
    Scale Color=&Multiply_Float(
     A=&Lerp_Float(A=0.4, B=1, Alpha=$System.Timeline.Time)
     B=&Lerp_Float(A=0.4, B=1, Alpha=$System.Waveform.Waveform)
    )
   )
  ]
  rendererClasses [&NiagaraMeshRendererProperties, &NiagaraLightRendererProperties]
  renderers [
   {rendererClass=&NiagaraMeshRendererProperties, rendererIndex=0}
   {rendererClass=&NiagaraMeshRendererProperties, rendererIndex=1}
   {rendererClass=&NiagaraMeshRendererProperties, rendererIndex=2}
   {rendererClass=&NiagaraMeshRendererProperties, rendererIndex=3}
   {rendererClass=&NiagaraMeshRendererProperties, rendererIndex=4}
   {rendererClass=&NiagaraMeshRendererProperties, rendererIndex=5}
   {rendererClass=&NiagaraMeshRendererProperties, rendererIndex=6}
   {rendererClass=&NiagaraLightRendererProperties, rendererIndex=7}
  ]
 }
 HandsArc {
  emitterUpdateScript [
   &EmitterState(
    Life Cycle Mode=::NewEnumerator1
    Loop Delay=&Lerp_Float(A=&RandomRangeFloat(Minimum=0.2, Maximum=0.05), B=0, Alpha=$System.Timeline.Time)
    Loop Duration=&Lerp_Float(A=&RandomRangeFloat(Minimum=0.1, Maximum=0.05), B=0.05, Alpha=$System.Timeline.Time)
    Recalculate Duration Each Loop=true
    UseLoopDelay=true
   ) {module=EmitterState001}
   &SpawnBurst_Instantaneous(
    Spawn Count=2
    Spawn Time=&RandomRangeFloat(Minimum=0, Maximum=0.05)
    Use Spawn Probability=false
    UseLoopCountLimit=false
   )
   &SpawnBurst_Instantaneous(
    Spawn Time=&RandomRangeFloat(Minimum=0.02, Maximum=0.05)
    Use Spawn Probability=false
    UseLoopCountLimit=false
   ) {module=SpawnBurst_Instantaneous001}
   &SpawnBurst_Instantaneous(
    Spawn Time=&RandomRangeFloat(Minimum=0.05, Maximum=0.1)
    Use Spawn Probability=false
    UseLoopCountLimit=false
   ) {module=SpawnBurst_Instantaneous002}
   &SpawnBurst_Instantaneous(
    Spawn Time=&RandomRangeFloat(Minimum=0.05, Maximum=0.13)
    Use Spawn Probability=false
    UseLoopCountLimit=false
   ) {module=SpawnBurst_Instantaneous003}
  ]
  particleSpawnScript [
   &PartitionParticles()
   &PartitionParticles(Number Of Partitions=4, Particles Per Partition=2, Partition Mode:ENiagara_PartitionMode=::NewEnumerator1) {module=PartitionParticles001}
   &InitializeParticle(
    Color=&MultiplyLinearColor(
     Linear Color=&RandomRangeLinearColor(Minimum=<rgba 1 1 1 1>, Maximum=<rgba 2 2 2 2>)
     Scale Factor=&MakeVector4_FromFloat(Float=0.8)
    )
    Lifetime=&RandomRangeFloat(Minimum=0.25, Maximum=0.4)
    Mesh Scale Max=<0.18 0.18 0.18>
    Mesh Scale Min=<0.12 0.12 0.12>
    Mesh Scale Mode=::NewEnumerator2
    Position Offset Coordinate Space=::Local
    Sprite Size Mode=::NewEnumerator3
    Uniform Sprite Size=50
    UsePositionOffset=true
    Write Material Random=true
   )
   &InitialMeshOrientation(
    Mesh Orientation Mode=::NewEnumerator7
    Rotation=&Add_Vector(
     A=&RandomRangeVector(Minimum=<0 0 0>, Maximum=<1 1 1>)
     B=&SelectVectorFromArray(
      Direct Array Index=$Particles.PartitionParticles001.Index In Partition
      Vector Selection Array=DataInterface() {propertyValues {FloatData [<0 0 0>, <0 0.25 0>]}}
     )
    )
    Use Rotation Vector=true
   )
   &AddVelocity(
    Velocity=&Multiply_Vector(
     A=&RandomRangeVector(Minimum=<0 0 0.8>, Maximum=<0 0 1>)
     B=&VectorFromFloat(Value=&SelectFloatFromArray(Direct Array Index=$Particles.PartitionParticles.Partition, Float Selection Array=DataInterface() {propertyValues {FloatData [-1,1]}}))
    )
    Velocity Speed Scale=455
   ) {enabled=false, module=AddVelocity002}
   &CameraOffset(Camera Offset Amount=20)
   &ShapeLocation(Box Size=<2 2 2>, MaskSpawnGroup=false, Override Seed=false, Shape Primitive=::NewEnumerator2, Surface Only=false, UseOwnerScale=true) {enabled=false}
  ]
  particleUpdateScript [
   &ParticleState()
   &SkeletalMeshLocation(
    Bone / Socket Index=$Particles.PartitionParticles001.Partition
    Bone Sampling Mode=::NewEnumerator4
    MaskSpawnGroup=false
    Sampled Position Offset=&RandomRangeVector(Minimum=<-3 -3 -3>, Maximum=<3 3 7>)
    Skeletal Mesh=$System.Skeletal Mesh
   )
   &ScaleMeshSize(Scale Factor=&VectorFromFloat(Value=&FloatFromCurve(
      FloatCurve=DataInterface() {propertyValues {Curve {keys [
          {arriveTangent=1.2917, interpMode=RCIM_Cubic, leaveTangent=1.2917, tangentMode=RCTM_User, time=0, value=0}
          {arriveTangent=0.0224, interpMode=RCIM_Cubic, leaveTangent=0.0224, tangentMode=RCTM_User, time=1, value=1}
         ]}}}
      Scale Curve=&MakeCustomFloatFromBool(Bool=$User.bIsFrontEndPreview, True Float=0.8, False Float=0.7)
     )))
   &ScaleMeshSize(Scale Factor=&RandomRangeVector(Minimum=<1 1 1.5>, Maximum=<1 1 2>)) {module=ScaleMeshSize001}
   &DynamicMaterialParameters(
    Index 0 Param 2=&RandomRangeFloat(Minimum=0, Maximum=0)
    Index 0 Param 4=13
    Index 1 Param 1=&FloatFromCurve(FloatCurve=DataInterface() {propertyValues {Curve {keys [<at 0 0.5>,<at 1 1>]}}})
    Index 1 Param 2=2
    Index 1 Param 3=&FloatFromCurve(FloatCurve=DataInterface() {propertyValues {Curve {keys [{interpMode=RCIM_Cubic, tangentMode=RCTM_Break, time=0.5, value=-0.5}, <at 1 0.1>]}}})
    Index 2 Param 1=&Multiply_Float(
     A=&FloatFromCurve(
      FloatCurve=DataInterface() {propertyValues {Curve {keys [{interpMode=RCIM_Cubic, leaveTangent=8.1085, tangentMode=RCTM_Break, time=0.8568, value=1}, {tangentMode=RCTM_User, time=1, value=0}]}}}
      Scale Curve=7
     )
     B=&Divide_Float(
      A=&MakeFloatFromVector(Vector=$Particles.Scale, Channel=::NewEnumerator0)
      B=&RandomRangeFloat(Minimum=10, Maximum=15)
     )
    )
    Index 2 Param 2=&FloatFromCurve(
     FloatCurve=DataInterface() {propertyValues {Curve {keys [
         {interpMode=RCIM_Cubic, tangentMode=RCTM_Break, time=0.7, value=0}
         {arriveTangent=-0.0859, interpMode=RCIM_Cubic, tangentMode=RCTM_Break, time=1, value=-1}
        ]}}}
     Scale Curve=50
    )
    Index 2 Param 3=&Waveform(
     Global Input Position=&RandomRangeFloat(Minimum=0, Maximum=15)
     Global Phase=$Particles.Age
     "Waveform [1]"=::NewEnumerator12
     "[1] Phase Scale"=2
     "[1] Frequency"=1
     "[1] Amplitude Min/Max"=<0 1>
    )
    Index 2 Param 4=25
    Index 3 Param 1=&FloatFromCurve(
     FloatCurve=DataInterface() {propertyValues {Curve {keys [
         {interpMode=RCIM_Cubic, leaveTangent=-0.0025, tangentMode=RCTM_Break, time=0, value=0}
         {arriveTangent=1.8528, interpMode=RCIM_Cubic, tangentMode=RCTM_Break, time=1, value=1}
        ]}}}
     Scale Curve=&RandomRangeFloat(Minimum=0, Maximum=0)
    )
    Index 3 Param 2=&FloatFromCurve(
     FloatCurve=DataInterface() {propertyValues {Curve {keys [
         {interpMode=RCIM_Cubic, leaveTangent=-0.0025, tangentMode=RCTM_Break, time=0, value=0.5}
         {arriveTangent=-0.0316, interpMode=RCIM_Cubic, tangentMode=RCTM_Break, time=1, value=-0.2}
        ]}}}
     Scale Curve=0
    )
    Index 3 Param 3=&FloatFromCurve(FloatCurve=DataInterface() {propertyValues {Curve {keys [
         {interpMode=RCIM_Cubic, leaveTangent=1.1034, tangentMode=RCTM_Break, time=0, value=-1}
         {arriveTangent=2.2399, tangentMode=RCTM_User, time=0.5, value=0.125}
        ]}}})
    Index 3 Param 4=&FloatFromCurve(FloatCurve=DataInterface() {propertyValues {Curve {keys [
         {interpMode=RCIM_Cubic, leaveTangent=1.8089, tangentMode=RCTM_Break, time=0.3, value=-0.5}
         {arriveTangent=2.0757, interpMode=RCIM_Cubic, tangentMode=RCTM_Break, time=0.9, value=1}
        ]}}})
    Module.Index0Param0Write=true
    Module.Index0Param1Write=true
    Module.Index0Param2Write=true
    Module.Index0Param3Write=true
    Module.Index1Param0Write=true
    Module.Index1Param1Write=true
    Module.Index1Param2Write=true
    Module.Index1Param3Write=true
    Module.Index2Param0Write=true
    Module.Index2Param1Write=true
    Module.Index2Param2Write=true
    Module.Index2Param3Write=true
    Module.Index3Param0Write=true
    Module.Index3Param1Write=true
    Module.Index3Param2Write=true
    Module.Index3Param3Write=true
    Write Parameter Index 1=true
    Write Parameter Index 2=true
    Write Parameter Index 3=true
   )
   &SolveForcesAndVelocity(Clamp Velocity=false, Limit Acceleration=false)
   &ScaleColor(
    Scale Alpha=&FloatFromCurve(
     FloatCurve=DataInterface() {propertyValues {Curve {keys [
         {tangentMode=RCTM_User, time=0, value=1}
         {arriveTangent=-1, interpMode=RCIM_Cubic, leaveTangent=-1, time=0.0834, value=1}
         {arriveTangent=-0.2668, interpMode=RCIM_Cubic, leaveTangent=-0.2668, tangentMode=RCTM_User, time=1, value=0}
        ]}}}
     Scale Curve=&Lerp_Float(A=0.2, B=0.5, Alpha=$System.Waveform.Waveform)
    )
    Scale RGB=&Lerp_Vector(
     A=&VectorFromFloat(Value=&FloatFromCurve(FloatCurve=DataInterface() {propertyValues {Curve {keys [<at 0.2771 1>,<at 0.3786 1>]}}}, Scale Curve=20))
     B=&VectorFromFloat(Value=&RandomRangeFloat(Minimum=0, Maximum=2, Evaluation Type=::NewEnumerator1))
     Alpha=&FloatFromCurve(FloatCurve=DataInterface() {propertyValues {Curve {keys [
          {interpMode=RCIM_Cubic, leaveTangent=2.5294, tangentMode=RCTM_Break, time=0, value=0}
          {arriveTangent=-0.2166, interpMode=RCIM_Cubic, tangentMode=RCTM_Break, time=0.4, value=1}
         ]}}})
    )
    ScaleA=true
    ScaleRGB=true
   )
   &FN_Sprite_Attributes(
    Color=&MultiplyLinearColor(
     Linear Color=<rgba 0.0854 0.0267 0.2 0.5>
     Scale Factor=&MakeVector4_FromLinearColor(LinearColor=$Particles.Color)
    )
    Create Color Attribute=true
    Write to Module Namespace=true
   )
   &FN_Sprite_Attributes(
    Color=&MultiplyLinearColor(
     Linear Color=<rgba 0.4768 0.2543 1 1>
     Scale Factor=&MakeVector4_FromLinearColor(LinearColor=$Particles.Color)
    )
    Create Color Attribute=true
    Create SpriteSize Attribute=true
    SpriteSize=&Vector2DFromFloat(Value=&FloatFromCurve(
      FloatCurve=DataInterface() {propertyValues {Curve {keys [
          {arriveTangent=4.1785, interpMode=RCIM_Cubic, leaveTangent=4.1785, tangentMode=RCTM_User, time=0, value=0}
          {arriveTangent=1.1179, interpMode=RCIM_Cubic, leaveTangent=1.1179, tangentMode=RCTM_User, time=0.2042, value=0.6228}
          {arriveTangent=0.0224, interpMode=RCIM_Cubic, leaveTangent=0.0224, tangentMode=RCTM_User, time=1, value=1}
         ]}}}
      Scale Curve=80
     ))
    Write to Module Namespace=true
   ) {module=FN_Sprite_Attributes001}
   /SetVariables() {isSetParameters=true}
  ]
  rendererClasses [&NiagaraMeshRendererProperties]
  renderers [
   {rendererClass=&NiagaraMeshRendererProperties, rendererIndex=0}
   {rendererClass=&NiagaraMeshRendererProperties, rendererIndex=1}
  ]
 }
 Line {
  emitterUpdateScript [
   &EmitterState(Inactive Response=::NewEnumerator1, Life Cycle Mode=::NewEnumerator1, Loop Behavior=::NewEnumerator1, Loop Duration=1, UseLoopDelay=false)
   &SpawnBurst_Instantaneous(Spawn Count=50, Use Spawn Probability=false, UseLoopCountLimit=false)
  ]
  particleSpawnScript [
   &InitializeParticle(
    Color Maximum=<rgba 6.1318 0.625 20 1>
    Color Minimum=<rgba 0.0573 0 1 1>
    Color Mode=::NewEnumerator2
    Lifetime Max=0.35
    Lifetime Min=0.25
    Lifetime Mode=::NewEnumerator1
    Sprite Size Max=<2.5 60>
    Sprite Size Min=<1.5 30>
    Sprite Size Mode=::NewEnumerator2
    UsePositionOffset=true
    Write Material Random=true
   )
   &ShapeLocation(
    Hemisphere Distribution=<1 0.7>
    MaskSpawnGroup=false
    Non Uniform Scale=<1 1 0.5>
    Offset=<0 0 80>
    Offset Mode=::NewEnumerator0
    Override Seed=false
    Sphere Radius=80
    UseOwnerScale=true
   )
   &AddVelocity(
    Module.Use Falloff Distance=false
    Override Seed=false
    Velocity Mode=::NewEnumerator1
    Velocity Origin=&AddVectorToPosition(Position=$Engine.Emitter.SimulationPosition, Vector=$System.GlobalOffset)
    Velocity Speed=&RandomRangeFloat(Minimum=600, Maximum=1000)
   )
   &AddVelocity(Velocity=&RandomRangeVector(Minimum=<0 0 200>, Maximum=<0 0 300>)) {module=AddVelocity001}
  ]
  particleUpdateScript [
   &ParticleState()
   &ScaleSpriteSize(Uniform Curve Sprite Scale=DataInterface() {propertyValues {Curve {keys [
        {arriveTangent=8.2127, interpMode=RCIM_Cubic, leaveTangent=8.2127, tangentMode=RCTM_User, time=0, value=0.5}
        {
         arriveTangent=4.2448
         arriveTangentWeight=0.232
         interpMode=RCIM_Cubic
         leaveTangent=4.2448
         leaveTangentWeight=2.2538
         tangentMode=RCTM_User
         time=0.07
         value=1
        }
        {arriveTangent=-2.5194, interpMode=RCIM_Cubic, leaveTangent=-2.5194, tangentMode=RCTM_User, time=0.75, value=0.8}
        {arriveTangent=-4.4982, interpMode=RCIM_Cubic, leaveTangent=-4.4982, tangentMode=RCTM_User, time=1, value=0}
       ]}}})
   &Drag(Drag=&RandomRangeFloat(Minimum=2, Maximum=4), Use Linear Drag=true, Use Rotational Drag=true)
   &SolveForcesAndVelocity(Clamp Velocity=false, Limit Acceleration=false)
   &ScaleColor(
    Scale Alpha=&Lerp_Float(
     A=0
     B=1
     Alpha=&FloatFromCurve(
      FloatCurve=DataInterface() {propertyValues {Curve {keys [
          {interpMode=RCIM_Cubic, tangentMode=RCTM_User, time=0, value=1}
          {interpMode=RCIM_Cubic, tangentMode=RCTM_User, time=0.5, value=1}
          {tangentMode=RCTM_User, time=1, value=0}
         ]}}}
      Scale Curve=2
     )
    )
    Scale RGB=&VectorFromFloat(Value=&FloatFromCurve(
      FloatCurve=DataInterface() {propertyValues {Curve {keys [
          {arriveTangent=-2.3805, interpMode=RCIM_Cubic, leaveTangent=-2.3805, tangentMode=RCTM_User, time=0, value=1}
          {interpMode=RCIM_Cubic, tangentMode=RCTM_User, time=1, value=0}
         ]}}}
      Scale Curve=&RandomRangeFloat(Minimum=1, Maximum=1.5)
     ))
    ScaleA=true
    ScaleRGB=true
   )
  ]
  rendererClasses [&NiagaraSpriteRendererProperties]
  renderers [{rendererClass=&NiagaraSpriteRendererProperties}]
 }
 RimArcs {
  emitterUpdateScript [
   &EmitterState(
    Life Cycle Mode=::NewEnumerator1
    Loop Delay=&Lerp_Float(A=&RandomRangeFloat(Minimum=0.3, Maximum=0.1), B=0, Alpha=$System.Timeline.Time)
    Loop Duration=&Lerp_Float(
     A=&RandomRangeFloat(Minimum=0.2, Maximum=0.7)
     B=&RandomRangeFloat(Minimum=0.2, Maximum=0.7)
     Alpha=$System.Timeline.Time
    )
    Recalculate Duration Each Loop=true
    UseLoopDelay=true
   )
   &SpawnBurst_Instantaneous(Spawn Count=&RandomRangeInt(Minimum=1, Maximum=1), Use Spawn Probability=true, UseLoopCountLimit=false) {module=SpawnBurst_Instantaneous001}
   &SpawnBurst_Instantaneous(
    Spawn Count=&RandomRangeInt(Minimum=1, Maximum=2)
    Spawn Time=0.08
    Use Spawn Probability=true
    UseLoopCountLimit=false
   ) {module=SpawnBurst_Instantaneous002}
   &SpawnBurst_Instantaneous(
    Spawn Count=&RandomRangeInt(Minimum=1, Maximum=2)
    Spawn Time=0.14
    Use Spawn Probability=true
    UseLoopCountLimit=false
   ) {module=SpawnBurst_Instantaneous003}
  ]
  particleSpawnScript [
   /SetVariables(Particles.PerParticleRandom:NiagaraFloat=&UniformAOrBFloat(A=0, Distribution Weight=0.3, Evaluation Type=::NewEnumerator0, Recalculate Random Each Loop=true, Override Seed=false)) {isSetParameters=true}
   &InitializeParticle(
    AdjustAlpha=true
    AdjustHue=true
    AdjustSaturation=false
    AdjustValue=true
    Alpha Scale Range:Vector2f=<0.5 1>
    Color=<rgba 0.8027 0.4858 1 1>
    Color Mode=::NewEnumerator3
    Hue Shift Range:Vector2f=<-0.1 40>
    Lifetime=&Lerp_Float(A=0.23, B=0.5, Alpha=$Particles.PerParticleRandom)
    Mesh Renderer Array Visibility Mode=::NewEnumerator3
    Mesh Renderer Info=DataInterface() {propertyValues {MeshRenderer=&self:RimArcs_0.NiagaraMeshRendererProperties_0}}
    Mesh Scale Mode=::NewEnumerator3
    Mesh Uniform Scale=0.95
    Position Offset=$System.SampleSkeletalMesh.SampledPosition
    UsePositionOffset=true
    Value Range:Vector2f=<0.4 1>
    Write Material Random=true
   )
   &InitialMeshOrientation(
    Mesh Axis To Orient=<0 1 0>
    Mesh Orientation Mode=::NewEnumerator2
    Orientation Coordinate Space=::Local
    Orientation Vector=&ReturnCameraVector(Camera Vectors=::NewEnumerator1, Coordinate Space=::Local)
    Rotation=&MakeVector(
     X=&RandomRangeFloat(Minimum=0, Maximum=0.02)
     Y=&Add_Float(
      A=&RandomRangeFloat(Minimum=0, Maximum=1)
      B=&ReturnNormalizedExecIndex(Normalized Index Scale=&RandomRangeFloat(Minimum=0.5, Maximum=0.5))
     )
     Z=&RandomRangeFloat(Minimum=-0.1, Maximum=0.1)
    )
    Use Rotation Vector=true
   ) {module=InitialMeshOrientation002}
  ]
  particleUpdateScript [
   &ParticleState()
   /SetVariables(Particles.Position:Vector3f=$System.SampleSkeletalMesh.SampledPosition) {isSetParameters=true}
   &UpdateMeshOrientation(Rotation Rate=&Lerp_Float(A=0.05, B=0, Alpha=$System.Timeline.Time), Rotation Vector=<0 -1 0>)
   &SolveForcesAndVelocity(Clamp Velocity=false, Limit Acceleration=false)
   &ScaleColor(
    Scale Alpha=&FloatFromCurve(FloatCurve=DataInterface() {propertyValues {Curve {keys [
         {interpMode=RCIM_Cubic, tangentMode=RCTM_User, time=0.25, value=1}
         {arriveTangent=-2.7545, interpMode=RCIM_Cubic, leaveTangent=-2.7545, tangentMode=RCTM_User, time=1, value=0}
        ]}}})
    Scale RGB=&Multiply_VectorByFloat(
     Vector=&MakeVectorFromLinearColorRGB(Color=&ColorFromCurve(ColorCurve=DataInterface() {propertyValues {
         AlphaCurve {keys [<at 0 1 RCIM_Cubic>,<at 1 0 RCIM_Cubic>]}
         BlueCurve {keys [
           <at 0.1176 6 RCIM_Cubic>
           <at 0.2526 0.3>
           <at 0.3091 400>
           <at 0.3547 0.3>
           {arriveTangent=0.3297, interpMode=RCIM_Cubic, leaveTangent=0.3297, time=0.4313, value=2}
           <at 0.9613 0.5>
          ]}
         GreenCurve {keys [
           <at 0.1176 6 RCIM_Cubic>
           <at 0.2526 0.3>
           <at 0.3091 400>
           <at 0.3547 0.3>
           {arriveTangent=0.3297, interpMode=RCIM_Cubic, leaveTangent=0.3297, time=0.4313, value=2}
           <at 0.9613 0.5>
          ]}
         RedCurve {keys [
           <at 0.1176 6 RCIM_Cubic>
           <at 0.2526 0.3>
           <at 0.3091 400>
           <at 0.3547 0.3>
           {arriveTangent=0.3297, interpMode=RCIM_Cubic, leaveTangent=0.3297, time=0.4313, value=2}
           <at 0.9613 0.5>
          ]}
        }}))
     Float=&MakeFloatFromVector(
      Vector=&VectorFromFloat(Value=&RandomRangeFloat(Minimum=0.1, Maximum=2, Evaluation Type=::NewEnumerator1))
      Channel=::NewEnumerator0
     )
    )
    ScaleA=true
    ScaleRGB=true
   )
   /SetVariables(Particles.MeshIndex=&ModuloInt(
     A=&Add_Integer(
      A=&RandomRangeInt(Minimum=0, Maximum=10)
      B=&MakeIntFromFloat(Float=&Lerp_Float(A=0, B=0, Alpha=$Particles.NormalizedAge), Conversion Mode=::NewEnumerator3)
     )
     B=6
    )) {isSetParameters=true}
   /SetVariables(
    Particles.LightningColor:LinearColor=&MultiplyLinearColor(
     Linear Color=&HueShiftLinearColor(
      Hue=&Lerp_Float(A=0.05, B=0.02, Alpha=$Particles.PerParticleRandom)
      Saturation=1
      Value=&Lerp_Float(A=0.7, B=1, Alpha=$Particles.PerParticleRandom)
      Linear Color=<rgba 1.6578 0 4 1>
     )
     Scale Factor=&MakeVector4_FromLinearColor(LinearColor=&ColorFromCurve(ColorCurve=DataInterface() {propertyValues {
         AlphaCurve {keys [<at 0 0>,<at 0.0494 1 RCIM_Cubic>,<at 0.7821 0 RCIM_Cubic>]}
         BlueCurve {keys [
           {arriveTangent=1.1002, interpMode=RCIM_Cubic, time=0.0699, value=10}
           <at 0.1016 0>
           <at 0.1818 0.2>
           <at 0.2228 3>
           <at 0.2619 0>
           <at 0.3645 3>
           <at 1 1 RCIM_Cubic>
          ]}
         GreenCurve {keys [
           {arriveTangent=1.1002, interpMode=RCIM_Cubic, time=0.0699, value=10}
           <at 0.1016 0>
           <at 0.1818 0.2>
           <at 0.2228 3>
           <at 0.2619 0>
           <at 0.3645 3>
           <at 1 1 RCIM_Cubic>
          ]}
         RedCurve {keys [
           {arriveTangent=1.1002, interpMode=RCIM_Cubic, time=0.0699, value=10}
           <at 0.1016 0>
           <at 0.1818 0.2>
           <at 0.2228 3>
           <at 0.2619 0>
           <at 0.3645 3>
           <at 1 1 RCIM_Cubic>
          ]}
        }}))
    )
    Particles.LightningDY0:Vector4f=&MakeVector4(
     X=&FloatFromCurve(
      FloatCurve=DataInterface() {propertyValues {Curve {keys [<at 0 1>,<at 1 0.5>]}}}
      Scale Curve=&Lerp_Float(A=0.5, B=1.2, Alpha=$Particles.PerParticleRandom)
     )
     Y=0
     Z=&RandomRangeFloat(Minimum=0, Maximum=22)
     W=&RandomRangeFloat(Minimum=0.5, Maximum=0.6)
    )
    Particles.LightningDY1:Vector4f=&MakeVector4(
     X=&FloatFromCurve(
      FloatCurve=DataInterface() {propertyValues {Curve {keys [
          <at 0 0>
          {arriveTangent=1, interpMode=RCIM_Cubic, leaveTangent=1, time=0.3, value=0}
          {interpMode=RCIM_Cubic, leaveTangent=-8.0089, time=1, value=1}
         ]}}}
      Scale Curve=0.02
     )
     Y=1
     Z=&FloatFromCurve(FloatCurve=DataInterface() {propertyValues {Curve {keys [<at 0.5 1 RCIM_Cubic>,<at 1 0 RCIM_Cubic>]}}})
     W=0
    )
    Particles.LightningDY2:Vector4f=&MakeVector4(
     X=&FloatFromCurve(FloatCurve=DataInterface() {propertyValues {Curve {keys [{tangentMode=RCTM_User, time=0, value=0}, {tangentMode=RCTM_User, time=1, value=1}]}}})
     Y=2
     Z=0
     W=0
    )
    Particles.LightningDY3:Vector4f=&MakeVector4(
     X=&FloatFromCurve(FloatCurve=DataInterface() {propertyValues {Curve {keys [{tangentMode=RCTM_User, time=0, value=0}, {tangentMode=RCTM_User, time=1, value=1}]}}})
     Y=1
     Z=1
     W=400
    )
    Particles.LightningLength:NiagaraFloat=&FloatFromCurve(
     FloatCurve=DataInterface() {propertyValues {Curve {keys [<at 0 0.9>,<at 1 1>]}}}
     Scale Curve=&Lerp_Float(A=1, B=1.4, Alpha=0)
    )
    Particles.LightningOrientation:Quat4f=&OrientMeshToVector(Look At Direction=<0 0 1>, Reference Axis Vector=<1 0 0>)
    Particles.LightningShapeExpand:NiagaraFloat=&FloatFromCurve(
     FloatCurve=DataInterface() {propertyValues {Curve {keys [
         {arriveTangent=4.1785, interpMode=RCIM_Cubic, leaveTangent=4.1785, tangentMode=RCTM_User, time=0, value=0}
         {arriveTangent=1.1179, interpMode=RCIM_Cubic, leaveTangent=1.1179, tangentMode=RCTM_User, time=0.2042, value=0.6228}
         {arriveTangent=0.0224, interpMode=RCIM_Cubic, leaveTangent=0.0224, tangentMode=RCTM_User, time=1, value=1}
        ]}}}
     Scale Curve=&Lerp_Float(A=0.5, B=0.35, Alpha=0)
    )
   ) {isSetParameters=true}
   /SetVariables(Particles.LightningSize:Vector3f=&Multiply_VectorByFloat(
     Vector=&Lerp_Vector(
      A=&MakeVector(
       X=&FloatFromCurve(FloatCurve=DataInterface() {propertyValues {Curve {keys [{arriveTangent=4.1785, interpMode=RCIM_Cubic, time=0, value=0.8}, {interpMode=RCIM_Cubic, leaveTangent=0.0224, time=1, value=1}]}}})
       Y=1
       Z=&FloatFromCurve(FloatCurve=DataInterface() {propertyValues {Curve {keys [<at 0 0.9>,<at 1 1>]}}})
      )
      B=<1 1 1>
      Alpha=&Lerp_Float(A=0.7, B=0, Alpha=$System.Timeline.Time)
     )
     Float=1
    )) {isSetParameters=true}
   /SetVariables(Particles.MeshOrientation=&MultiplyQuaternion(
     Quaternion A=$Particles.Initial.MeshOrientation
     Quaternion B=&OrientMeshToVector(
      Look At Direction=&ReturnCameraVector(Camera Vectors=::NewEnumerator1, Coordinate Space=::Local)
      Reference Axis Vector=<1 0 0>
     )
    )) {enabled=false, isSetParameters=true}
  ]
  rendererClasses [&NiagaraMeshRendererProperties]
  renderers [{rendererClass=&NiagaraMeshRendererProperties, rendererIndex=0}]
 }
 Streak {
  emitterUpdateScript [
   &EmitterState(Inactive Response=::NewEnumerator1, Life Cycle Mode=::NewEnumerator1, Loop Behavior=::NewEnumerator1, Loop Duration=1, UseLoopDelay=false)
   &SpawnBurst_Instantaneous(Spawn Count=10, Spawn Time=0.05, Use Spawn Probability=false, UseLoopCountLimit=false)
  ]
  particleSpawnScript [
   &InitializeParticle(
    Color Minimum=<rgba 0.1 0.1 0.1 2>
    Color Mode=::NewEnumerator2
    Lifetime Max=0.4
    Lifetime Min=0.25
    Lifetime Mode=::NewEnumerator1
    Sprite Size Max=<10 20>
    Sprite Size Min=<5 10>
    Sprite Size Mode=::NewEnumerator2
    Sprite UV Mode=::NewEnumerator1
    UsePositionOffset=true
    Write Material Random=true
   )
   &CameraOffset(Camera Offset Amount=&SelectFloatFromArray(Direct Array Index=&SpawnGroup(), Float Selection Array=DataInterface() {propertyValues {FloatData [10,10]}}))
   &ShapeLocation(
    Handle Radius=5
    Large Radius=5
    MaskSpawnGroup=false
    Offset=<0 0 80>
    Offset Mode=::NewEnumerator0
    Override Seed=false
    Rotation Mode=::NewEnumerator1
    Shape Primitive=::NewEnumerator3
    UseOwnerScale=true
   )
   &AddVelocity(
    Module.Use Falloff Distance=false
    Override Seed=false
    Velocity Mode=::NewEnumerator1
    Velocity Origin=$Particles.ShapeLocation.ShapePosition
    Velocity Speed=&RandomRangeFloat(Minimum=400, Maximum=600)
   ) {module=AddVelocity002}
  ]
  particleUpdateScript [
   &ParticleState()
   &ScaleSpriteSize(
    Scale Factor=&Multiply_Vector2D(A=&RandomRangeVector2D(Minimum=<1.2 2>, Maximum=<3 3.5>), B=<1 1>)
    Scale Sprite Size Mode=::NewEnumerator2
   )
   &ScaleSpriteSize(
    Non-Uniform Curve Scale=<1 2>
    Non-Uniform Curve Sprite Scale=DataInterface() {propertyValues {
      XCurve {keys [
        {arriveTangent=0.8609, interpMode=RCIM_Cubic, leaveTangent=0.8609, tangentMode=RCTM_User, time=0, value=0.6}
        {arriveTangent=0.5103, interpMode=RCIM_Cubic, leaveTangent=0.5103, tangentMode=RCTM_User, time=0.3, value=0.8}
        {arriveTangent=0.0224, interpMode=RCIM_Cubic, leaveTangent=0.0224, tangentMode=RCTM_User, time=1, value=1}
       ]}
      YCurve {keys [
        {arriveTangent=0.8801, interpMode=RCIM_Cubic, leaveTangent=0.8801, tangentMode=RCTM_User, time=0, value=0.3}
        {arriveTangent=0.6586, interpMode=RCIM_Cubic, leaveTangent=0.6586, tangentMode=RCTM_User, time=0.2, value=0.85}
        {arriveTangent=-0.4778, interpMode=RCIM_Cubic, leaveTangent=-0.4778, tangentMode=RCTM_User, time=1, value=0.75}
       ]}
     }}
    Scale Sprite Size Mode=::NewEnumerator3
   ) {module=ScaleSpriteSize001}
   &ScaleColor(
    Scale Alpha=&FloatFromCurve(
     FloatCurve=DataInterface() {propertyValues {Curve {keys [
         <at 0 0 RCIM_Cubic>
         {arriveTangent=-1.2233, interpMode=RCIM_Cubic, leaveTangent=-1.2233, tangentMode=RCTM_User, time=0.1, value=1}
         <at 1 0 RCIM_Cubic>
        ]}}}
     Scale Curve=&Multiply_Float(A=2, B=&FNTimeOfDayScalar(Night Scale=0.6, Morning Scale=0.8, Evening Scale=0.6))
    )
    Scale RGB=&MakeVectorFromLinearColorRGB(Color=&ColorFromCurve(ColorCurve=DataInterface() {propertyValues {
        AlphaCurve {keys []}
        BlueCurve {keys [<at 0.1191 10 RCIM_Cubic>,<at 0.7128 0.0573 RCIM_Cubic>]}
        GreenCurve {keys [<at 0.1191 0.0521 RCIM_Cubic>,<at 0.7128 0.0061 RCIM_Cubic>]}
        RedCurve {keys [<at 0.1191 1.9127 RCIM_Cubic>,<at 0.7128 0.0415 RCIM_Cubic>]}
       }}))
    ScaleA=true
    ScaleRGB=true
   )
   &Drag(Drag=3, Use Linear Drag=true, Use Rotational Drag=true)
   &SolveForcesAndVelocity(Clamp Velocity=false, Limit Acceleration=false)
   &DynamicMaterialParameters(
    Index 0 Param 1=&FloatFromCurve(FloatCurve=DataInterface() {propertyValues {Curve {keys [
         {arriveTangent=-0.1846, time=0, value=0}
         {arriveTangent=1.3385, interpMode=RCIM_Cubic, leaveTangent=1.1594, tangentMode=RCTM_Break, time=0.5, value=0}
         {arriveTangent=2, interpMode=RCIM_Cubic, leaveTangent=2, time=0.8, value=0.5}
         {arriveTangent=3.0091, interpMode=RCIM_Cubic, leaveTangent=3.0091, tangentMode=RCTM_User, time=1, value=1}
        ]}}})
    Index 0 Param 2=&FloatFromCurve(FloatCurve=DataInterface() {propertyValues {Curve {keys [
         {arriveTangent=-0.0189, interpMode=RCIM_Cubic, leaveTangent=-0.0189, tangentMode=RCTM_User, time=0, value=1}
         {arriveTangent=-3.722, interpMode=RCIM_Cubic, leaveTangent=-3.722, tangentMode=RCTM_User, time=0.8, value=0}
        ]}}})
    Index 0 Param 4=&FloatFromCurve(FloatCurve=DataInterface() {propertyValues {Curve {keys [<at 0 0 RCIM_Cubic>,<at 1 1 RCIM_Cubic>]}}}, Scale Curve=0.8)
    Module.Index0Param0Write=true
    Module.Index0Param1Write=true
    Module.Index0Param2Write=true
    Module.Index0Param3Write=true
   )
  ]
  rendererClasses [&NiagaraSpriteRendererProperties]
  renderers [{rendererClass=&NiagaraSpriteRendererProperties}]
 }
 Thunder {
  emitterUpdateScript [
   &EmitterState(Life Cycle Mode=::NewEnumerator1, Loop Duration=1, UseLoopDelay=false)
   &SpawnRate(Spawn Probability=0.6, SpawnRate=&Lerp_Float(A=0, B=5, Alpha=$System.Timeline.Time), Use Spawn Probability=true)
  ]
  particleSpawnScript [
   &InitializeParticle(
    Color=&RandomRangeLinearColor(Minimum=<rgba 2.6786 0 10 1>, Maximum=<rgba 0.1 0 0.2 1>)
    Lifetime=&Multiply_Float(
     A=&Lerp_Float(A=2, B=1, Alpha=$System.Timeline.Time)
     B=&RandomRangeFloat(Minimum=0.1, Maximum=0.25)
    )
    Mesh Renderer Array Visibility Mode=::NewEnumerator3
    Mesh Renderer Info=DataInterface() {propertyValues {MeshRenderer=&self:Thunder_2.NiagaraMeshRendererProperties_1}}
    Mesh Scale=&Multiply_Vector(
     A=&RandomRangeVector(Minimum=<0.125 0.125 0.125>, Maximum=<0.15 0.15 0.175>)
     B=&MakeVector(X=1, Y=1, Z=&Lerp_Float(A=0.7, B=1.1, Alpha=0.5))
    )
    Mesh Scale Mode=::NewEnumerator4
    Sprite Rotation Mode=::NewEnumerator2
    Sprite Size Mode=::NewEnumerator3
    Uniform Sprite Size=30
    UsePositionOffset=false
    Write Material Random=true
   )
   /SetVariables(Particles.SpawnGroup=$Emitter.SpawnGroup) {isSetParameters=true}
   /SetVariables(Particles.VisibilityTag=$Particles.SpawnGroup) {isSetParameters=true}
   &PartitionParticles(
    Number Of Partitions=4
    Particles Per Partition:NiagaraInt32=&MaskIntBySpawnGroup(Masked Int=2, Passthrough Int=1, Spawn Group Mask=0)
    Partition Mode:ENiagara_PartitionMode=::NewEnumerator1
   )
   &InitialMeshOrientation(
    Mesh Orientation Mode=::NewEnumerator2
    Orientation Coordinate Space=::World
    Orientation Vector=<0 0 1>
    Rotation=&MakeVector(X=&RandomRangeFloat(Minimum=0, Maximum=360), Y=0, Z=0)
    Use Rotation Vector=true
   )
   /SetVariables(Particles.PerParticleRandom:NiagaraFloat=&RandomRangeFloat(Minimum=0, Maximum=1)) {isSetParameters=true}
  ]
  particleUpdateScript [
   &ParticleState()
   &ShapeLocation(
    MaskSpawnGroup=false
    Offset=&Lerp_Vector(A=<0 0 -50>, B=<0 0 -100>, Alpha=$Particles.PerParticleRandom)
    Offset Coordinate Space=::World
    Offset Mode=::NewEnumerator0
    Override Seed=false
    Ring / Disc Distribution Mode:ENiagaraSphereDistributionMode=::NewEnumerator1
    Ring Radius=&Lerp_Float(A=100, B=50, Alpha=$Particles.PerParticleRandom)
    Shape Origin=$System.SampleSkeletalMesh.SampledPosition
    Shape Primitive=::NewEnumerator4
    U Position=&RandomRangeFloat(Minimum=0, Maximum=1)
    UseOwnerScale=true
   )
   /SetVariables(
    Particles.LightningLength:NiagaraFloat=&FloatFromCurve(FloatCurve=DataInterface() {propertyValues {Curve {keys [<at 0 0.9>,<at 1 1>]}}})
    Particles.LightningShapeExpand:NiagaraFloat=&FloatFromCurve(FloatCurve=DataInterface() {propertyValues {Curve {keys [
         {arriveTangent=4.1785, interpMode=RCIM_Cubic, leaveTangent=4.1785, tangentMode=RCTM_User, time=0, value=0}
         {arriveTangent=1.1179, interpMode=RCIM_Cubic, leaveTangent=1.1179, tangentMode=RCTM_User, time=0.2042, value=0.6228}
         {arriveTangent=0.0224, interpMode=RCIM_Cubic, leaveTangent=0.0224, tangentMode=RCTM_User, time=1, value=1}
        ]}}})
   ) {isSetParameters=true}
   &ScaleMeshSize(Scale Factor=&MakeVector(
     X=$Particles.LightningShapeExpand
     Y=$Particles.LightningShapeExpand
     Z=&Multiply_Float(A=$Particles.LightningLength, B=&RandomRangeFloat(Minimum=2, Maximum=4))
    ))
   &DynamicMaterialParameters(
    Index 0 Param 1=&FloatFromCurve(
     FloatCurve=DataInterface() {propertyValues {Curve {keys [<at 0 1>,<at 1 0.5>]}}}
     Scale Curve=&RandomRangeFloat(Minimum=3, Maximum=0.5)
    )
    Index 0 Param 2=1
    Index 0 Param 3=&RandomRangeFloat(Minimum=0, Maximum=22)
    Index 0 Param 4=0.6
    Index 1 Param 1=&FloatFromCurve(
     FloatCurve=DataInterface() {propertyValues {Curve {keys [
         <at 0 0>
         {arriveTangent=1, interpMode=RCIM_Cubic, leaveTangent=1, time=0.3, value=0}
         {interpMode=RCIM_Cubic, leaveTangent=-8.0089, time=1, value=1}
        ]}}}
     Scale Curve=0.1
    )
    Index 1 Param 2=1
    Index 1 Param 3=&FloatFromCurve(FloatCurve=DataInterface() {propertyValues {Curve {keys [<at 0.5 1 RCIM_Cubic>,<at 1 0 RCIM_Cubic>]}}}, Scale Curve=2)
    Index 2 Param 1=&FloatFromCurve(
     FloatCurve=DataInterface() {propertyValues {Curve {keys [{tangentMode=RCTM_User, time=0, value=0}, {tangentMode=RCTM_User, time=1, value=1}]}}}
     Scale Curve=0
    )
    Index 2 Param 2=2
    Index 3 Param 1=1
    Index 3 Param 4=30
    Module.Index0Param0Write=true
    Module.Index0Param1Write=true
    Module.Index0Param2Write=true
    Module.Index0Param3Write=true
    Module.Index1Param0Write=true
    Module.Index1Param1Write=true
    Module.Index1Param2Write=true
    Module.Index1Param3Write=true
    Module.Index2Param0Write=true
    Module.Index2Param1Write=true
    Module.Index2Param2Write=true
    Module.Index2Param3Write=true
    Module.Index3Param0Write=true
    Module.Index3Param1Write=true
    Module.Index3Param2Write=true
    Module.Index3Param3Write=true
    Write Parameter Index 1=true
    Write Parameter Index 2=true
    Write Parameter Index 3=true
   )
   &ScaleColor(
    Scale Alpha=&FloatFromCurve(FloatCurve=DataInterface() {propertyValues {Curve {keys [
         {tangentMode=RCTM_User, time=0, value=1}
         {arriveTangent=-0.0405, interpMode=RCIM_Cubic, leaveTangent=-0.0405, tangentMode=RCTM_User, time=0.7009, value=1}
         {arriveTangent=-8.0089, interpMode=RCIM_Cubic, leaveTangent=-8.0089, tangentMode=RCTM_User, time=1, value=0}
        ]}}})
    Scale RGB=&VectorFromFloat(Value=&FloatFromCurve(
      FloatCurve=DataInterface() {propertyValues {Curve {keys [
          {arriveTangent=9.0725, interpMode=RCIM_Cubic, time=0, value=0}
          <at 0.1408 2 RCIM_Cubic>
          {interpMode=RCIM_Cubic, leaveTangent=-0.0405, time=1, value=0}
         ]}}}
      Scale Curve=&Multiply_Float(
       A=&FN_FrontEndFloatModifier(Gameplay=&RandomRangeFloat(Minimum=4, Maximum=2), Front End=8)
       B=1
      )
     ))
    ScaleA=true
    ScaleRGB=true
   ) {module=ScaleColor001}
   &CameraOffset(Camera Offset Amount=10)
   &FN_Sprite_Attributes(
    Color=&MultiplyLinearColor(Linear Color=$Particles.Color, Scale Factor=<0.2 0.2 0.2 1>)
    Create Color Attribute=true
    Write to Module Namespace=true
   )
  ]
  rendererClasses [&NiagaraMeshRendererProperties, &NiagaraSpriteRendererProperties]
  renderers [
   {rendererClass=&NiagaraMeshRendererProperties, rendererIndex=0}
   {rendererClass=&NiagaraSpriteRendererProperties, rendererIndex=1}
  ]
 }
 energy_loss_idle {
  emitterUpdateScript [
   &EmitterState(
    Inactive Response=::NewEnumerator1
    Life Cycle Mode=::NewEnumerator1
    Loop Behavior=::NewEnumerator1
    Loop Duration Mode=::NewEnumerator1
    UseLoopDelay=false
   )
   &SpawnBurst_Instantaneous(Use Spawn Probability=false, UseLoopCountLimit=true)
   /SetVariables(Emitter.PanSpeed=-0.25) {isSetParameters=true, module=SetVariables_#7}
  ]
  particleSpawnScript [
   &InitializeParticle(
    Color=<rgba 1 1 1 0.2>
    Lifetime=0.3
    Mesh Scale=&Multiply_VectorByFloat(
     Vector=<0.98 1.25 1.25>
     Float=&Multiply_Float(A=&Multiply_Float(A=0.01, B=0.8), B=$System.SphereDiameter)
    )
    Mesh Scale Mode=::NewEnumerator4
    Position Offset=&AddVectorToPosition(Position=<0 0 0>, Vector=$System.GlobalOffset)
    UsePositionOffset=true
    Write Material Random=true
   )
   &InitialMeshOrientation(Mesh Orientation Mode=::NewEnumerator2, Use Rotation Vector=true) {enabled=false}
   &CameraOffset(Camera Offset Amount=50) {enabled=false}
  ]
  particleUpdateScript [
   &ParticleState(Kill Particles When Lifetime Has Elapsed=false)
   /SetVariables(Particles.Position=$System.SampleSkeletalMesh.SampledPosition) {isSetParameters=true, module=SetVariables_#8}
   &UpdateMeshOrientation(Facing Interpolation=false, Orientation Method=::NewEnumerator1, Override Quat=false, TransformFacing=false, TransformUp=false) {enabled=false}
   &ScaleMeshSize(Scale Factor=&Multiply_VectorByFloat(
     Vector=<1 0.9 0.9>
     Float=&FloatFromCurve(
      FloatCurve=DataInterface() {propertyValues {Curve {keys [
          {arriveTangent=7.387, interpMode=RCIM_Cubic, leaveTangent=7.387, tangentMode=RCTM_User, time=0, value=0}
          {arriveTangent=1.65, interpMode=RCIM_Cubic, leaveTangent=1.65, time=0.2, value=1.025}
          {arriveTangent=0.2529, interpMode=RCIM_Cubic, leaveTangent=0.2529, tangentMode=RCTM_User, time=1, value=1.65}
         ]}}}
      Scale Curve=0.83
     )
    ))
   &SolveForcesAndVelocity(Clamp Velocity=false, Limit Acceleration=false)
   &ScaleColor(
    Scale Alpha=&Multiply_Float(
     A=&Lerp_Float(A=0.4, B=0, Alpha=$System.Timeline.Time)
     B=&FN_FrontEndFloatModifier(
      Gameplay=&Max_Float(
       A:NiagaraFloat=&DotProduct(Vector A=&ReturnCameraVector(Camera Vectors=::NewEnumerator1, Coordinate Space=::Local), Vector B=<0 1 0>)
       B:NiagaraFloat=0.2
      )
      Front End=0.5
     )
    )
    Scale RGB=&MakeVectorFromLinearColorRGB(Color=<rgba 1 0.1 6 1>)
    ScaleA=true
    ScaleRGB=true
   )
  ]
  rendererClasses [&NiagaraMeshRendererProperties]
  renderers [{rendererClass=&NiagaraMeshRendererProperties, rendererIndex=0}]
 }
 energy_loss_idle001 {
  emitterUpdateScript [
   &EmitterState(
    Inactive Response=::NewEnumerator1
    Life Cycle Mode=::NewEnumerator1
    Loop Behavior=::NewEnumerator1
    Loop Duration Mode=::NewEnumerator1
    UseLoopDelay=false
   )
   &SpawnBurst_Instantaneous(Use Spawn Probability=false, UseLoopCountLimit=true)
   /SetVariables(Emitter.PanSpeed=-4) {isSetParameters=true, module=SetVariables_#7}
  ]
  particleSpawnScript [
   &InitializeParticle(
    Color=<rgba 1 1 1 4>
    Lifetime=0.3
    Mesh Scale=&Multiply_VectorByFloat(
     Vector=<1.1 1.35 1.35>
     Float=&Multiply_Float(A=&Multiply_Float(A=0.01, B=0.8), B=$System.SphereDiameter)
    )
    Mesh Scale Mode=::NewEnumerator4
    Position Offset=&AddVectorToPosition(Position=<0 0 0>, Vector=$System.GlobalOffset)
    UsePositionOffset=true
    Write Material Random=true
   )
   &InitialMeshOrientation(Mesh Orientation Mode=::NewEnumerator2, Use Rotation Vector=true) {enabled=false}
   &CameraOffset(Camera Offset Amount=-10)
  ]
  particleUpdateScript [
   &ParticleState(Kill Particles When Lifetime Has Elapsed=false)
   /SetVariables(Particles.Position=$System.SampleSkeletalMesh.SampledPosition) {isSetParameters=true, module=SetVariables_#8}
   &UpdateMeshOrientation(Facing Interpolation=false, Orientation Method=::NewEnumerator1, Override Quat=false, TransformFacing=false, TransformUp=false) {enabled=false}
   &ScaleMeshSize(Scale Factor=&Multiply_VectorByFloat(
     Vector=<1 0.9 0.9>
     Float=&FloatFromCurve(
      FloatCurve=DataInterface() {propertyValues {Curve {keys [
          {arriveTangent=7.387, interpMode=RCIM_Cubic, leaveTangent=7.387, tangentMode=RCTM_User, time=0, value=0}
          {arriveTangent=1.65, interpMode=RCIM_Cubic, leaveTangent=1.65, time=0.2, value=1.025}
          {arriveTangent=0.2529, interpMode=RCIM_Cubic, leaveTangent=0.2529, tangentMode=RCTM_User, time=1, value=1.65}
         ]}}}
      Scale Curve=0.83
     )
    ))
   &SolveForcesAndVelocity(Clamp Velocity=false, Limit Acceleration=false)
   &ScaleColor(
    Scale Alpha=&Multiply_Float(
     A=&Lerp_Float(A=0, B=0.2, Alpha=$System.Timeline.Time)
     B=&FN_FrontEndFloatModifier(
      Gameplay=&Lerp_Float(
       A=0.05
       B=0.5
       Alpha=&ClampFloat(Float=&DotProduct(Vector A=&ReturnCameraVector(Camera Vectors=::NewEnumerator1, Coordinate Space=::Local), Vector B=<0 1 0>))
      )
      Front End=0.2
     )
    )
    Scale RGB=&MakeVectorFromLinearColorRGB(Color=<rgba 1 0.1 6 1>)
    ScaleA=true
    ScaleRGB=true
   )
  ]
  rendererClasses [&NiagaraMeshRendererProperties]
  renderers [{rendererClass=&NiagaraMeshRendererProperties, rendererIndex=0}]
 }
}
summary {
 systemName=NS_Snapfreeze_Contrail
 emitters [
  {emitterName=Dirt, rendererClasses [&NiagaraSpriteRendererProperties]}
  {emitterName=Streak, rendererClasses [&NiagaraSpriteRendererProperties]}
  {emitterName=Line, rendererClasses [&NiagaraSpriteRendererProperties]}
  {emitterName=Energy_Sphere, rendererClasses [&NiagaraMeshRendererProperties, &NiagaraLightRendererProperties]}
  {
   emitterName=EndpointsV2
   rendererClasses [
    &NiagaraRibbonRendererProperties
    &NiagaraMeshRendererProperties
    &NiagaraSpriteRendererProperties
   ]
  }
  {emitterName=Decal_, rendererClasses [&NiagaraSpriteRendererProperties]}
  {emitterName=energy_loss_idle, rendererClasses [&NiagaraMeshRendererProperties]}
  {emitterName=energy_loss_idle001, rendererClasses [&NiagaraMeshRendererProperties]}
  {emitterName=HandsArc, rendererClasses [&NiagaraMeshRendererProperties]}
  {emitterName=Thunder, rendererClasses [&NiagaraMeshRendererProperties, &NiagaraSpriteRendererProperties]}
  {emitterName=RimArcs, rendererClasses [&NiagaraMeshRendererProperties]}
 ]
 userVariables [
  User.bIsFrontEnd {type=NiagaraBool}
  User.bIsFrontEndPreview {type=NiagaraBool}
  User.ParaGlideLeanAlpha {type=NiagaraFloat, defaultValue=0, description=""}
 ]
}
